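# Physics asset for the F_MED_CameraShake female character's FaceAcc part, defining bodies for dynamic hair and an earring.
# Six hair chains (three left, one middle, two right) link capsules and spheres to small master-bone spheres with limited swing-twist constraints, with middle-chain twist locked.
# Looser cross-constraints tie the middle chain to left1 and right1.
# Head, neck, clavicle, upper spine and upper-arm shapes provide body volumes; the earring sphere has no constraint.
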PhysicsAsset /BRCosmetics/Characters/Player/Female/Medium/Bodies/F_MED_CameraShake/Meshes/Parts/F_MED_CameraShake_FaceAcc_Physics
bodies {
 clavicle_l [{shapeType=Capsule, length=9.25, radius=5.1, center=<6 -1.1533 -2>, rotation=<rot 80 -0.0004 -0.0005>}]
 clavicle_r [{shapeType=Capsule, length=9.25, radius=5.1, center=<-6 1.7361 2>, rotation=<rot 80 -0.0008 -0.0009>}]
 dyn_earring_01 [{shapeType=Sphere, radius=5, shapeName=dyn_earring_01_sphere, center=<0.4298 2.1381 0>}]
 dyn_hair_left01_03 [{shapeType=Capsule, length=5, radius=2.3, shapeName=dyn_hair_left01_03_capsule, center=<2 0 1>, rotation=<rot 90 0 0>}]
 dyn_hair_left1_01 [{shapeType=Capsule, length=2, radius=1, shapeName=dyn_hair_left1_01_capsule, center=<2 0 0>, rotation=<rot 90 0 0>}]
 dyn_hair_left1_02 [{shapeType=Capsule, length=4, radius=1.8, shapeName=dyn_hair_left1_02_capsule, center=<2.5 0 1>, rotation=<rot 90 0 0>}]
 dyn_hair_left2_01 [{shapeType=Capsule, length=2, radius=1, shapeName=dyn_hair_left2_01_capsule, center=<2 0 0>, rotation=<rot 90 0 0>}]
 dyn_hair_left2_02 [{shapeType=Capsule, length=2, radius=2, shapeName=dyn_hair_left2_02_capsule, center=<1 0 0>, rotation=<rot 90 0 0>}]
 dyn_hair_left2_03 [{shapeType=Sphere, radius=3, shapeName=dyn_hair_left2_03_sphere, center=<2.012 -1.0376 -1.7125>}]
 dyn_hair_left3_01 [{shapeType=Capsule, length=2, radius=1, shapeName=dyn_hair_left3_01_capsule, center=<2 0 0>, rotation=<rot 90 0 0>}]
 dyn_hair_left3_02 [{shapeType=Capsule, length=2, radius=2, shapeName=dyn_hair_left3_02_capsule, center=<2 0 0>, rotation=<rot 90 0 0>}]
 dyn_hair_left3_03 [{shapeType=Sphere, radius=3, shapeName=dyn_hair_left3_03_sphere, center=<3.3693 -0.1094 -2.0314>}]
 dyn_hair_midle_01 [{shapeType=Capsule, length=1.5, radius=1.5, shapeName=dyn_hair_midle_01_capsule, center=<2 0 0>, rotation=<rot 90 0 0>}]
 dyn_hair_midle_02 [{shapeType=Capsule, length=2, radius=1.7, shapeName=dyn_hair_midle_02_capsule, center=<1 0 0>, rotation=<rot 90 0 0>}]
 dyn_hair_midle_03 [
  {shapeType=Capsule, length=3.5, radius=1.5, shapeName=dyn_hair_midle_03_capsule, center=<1 0 0>, rotation=<rot 90 0 0>}
  {shapeType=Capsule, length=8, radius=1.6, shapeName=dyn_hair_midle_03_capsule, center=<5.4246 -0.0128 0.5649>, rotation=<rot 0 0 88.7067>}
 ]
 dyn_hair_midle_04 [{shapeType=Capsule, length=3.7, radius=2, shapeName=dyn_hair_midle_04_capsule, center=<2 0 0>, rotation=<rot 90 0 0>}]
 dyn_hair_right1_01 [{shapeType=Capsule, length=2, radius=1, shapeName=dyn_hair_right1_01_capsule, center=<2 0 0>, rotation=<rot 87.5376 0.0002 0.0002>}]
 dyn_hair_right1_02 [
  {
   shapeType=Capsule
   length=7
   radius=2.1
   shapeName=dyn_hair_right1_02_capsule
   center=<7.0174 -3.7801 -0.7976>
   rotation=<rot 57.1961 -75.9234 -132.7346>
  }
  {
   shapeType=Capsule
   length=5
   radius=0.5
   shapeName=dyn_hair_right1_02_capsule
   center=<0.0117 0.1175 -0.0037>
   rotation=<rot -85.1688 -163.8814 167.4405>
  }
 ]
 dyn_hair_right1_03 [{shapeType=Capsule, length=4, radius=2.4, shapeName=dyn_hair_right1_03_capsule, center=<4 0.7818 0>, rotation=<rot 90 0 0>}]
 dyn_hair_right2_01 [{
   shapeType=Capsule
   length=2
   radius=1
   shapeName=dyn_hair_right2_01_capsule
   center=<1.4412 -0.181 -0.0804>
   rotation=<rot 74.0082 -8.9068 4.7972>
  }]
 dyn_hair_right2_02 [{shapeType=Capsule, length=2.5, radius=2, shapeName=dyn_hair_right2_02_capsule, rotation=<rot 84.9037 -8.7798 -2.2326>}]
 head [{shapeType=Sphere, radius=7.5, shapeName=head_sphere}]
 mstr_hair_left [{shapeType=Sphere, radius=0.2, shapeName=mstr_hair_left_sphere}]
 mstr_hair_left2 [{shapeType=Sphere, radius=0.2, shapeName=mstr_hair_left_sphere}]
 mstr_hair_left3 [{shapeType=Sphere, radius=0.2, shapeName=mstr_hair_left_sphere}]
 mstr_hair_midle [{shapeType=Sphere, radius=0.2, shapeName=mstr_hair_left_sphere}]
 mstr_hair_right1 [{shapeType=Sphere, radius=0.2, shapeName=mstr_hair_left_sphere}]
 mstr_hair_right2 [{shapeType=Sphere, radius=0.2, shapeName=mstr_hair_left_sphere}]
 neck_02 [{shapeType=Capsule, length=8.5, radius=4, center=<-2.5005 -0.6731 0>, rotation=<rot 90 0 0>}]
 spine_04 [{shapeType=Capsule, length=9, radius=10, shapeName=spine_04_capsule, center=<0 0.7139 0>}]
 spine_05 [{shapeType=Capsule, length=14, radius=9.7, shapeName=spine_05_capsule, center=<0 -1.6569 0>}]
 upperarm_l [{shapeType=Sphere, radius=4, shapeName=upperarm_l_sphere}]
 upperarm_r [{shapeType=Sphere, radius=5.2, shapeName=upperarm_r_sphere}]
}
constraints [
 {bone1Name=dyn_hair_left1_02, bone2Name=dyn_hair_left1_01, swing1LimitDegrees=10, swing2LimitDegrees=30, twistLimitDegrees=10}
 {bone1Name=dyn_hair_left01_03, bone2Name=dyn_hair_left1_02, swing1LimitDegrees=10, swing2LimitDegrees=30, twistLimitDegrees=8}
 {bone1Name=dyn_hair_left2_02, bone2Name=dyn_hair_left2_01, swing1LimitDegrees=2, swing2LimitDegrees=2, twistLimitDegrees=15}
 {bone1Name=dyn_hair_left2_03, bone2Name=dyn_hair_left2_02, swing1LimitDegrees=8, swing2LimitDegrees=6, twistLimitDegrees=15}
 {bone1Name=dyn_hair_left3_02, bone2Name=dyn_hair_left3_01, swing1LimitDegrees=20, swing2LimitDegrees=30, twistLimitDegrees=5}
 {bone1Name=dyn_hair_left3_03, bone2Name=dyn_hair_left3_02, swing1LimitDegrees=45, swing2LimitDegrees=30, twistLimitDegrees=10}
 {
  bone1Name=dyn_hair_midle_02
  bone2Name=dyn_hair_midle_01
  swing1LimitDegrees=20
  swing2LimitDegrees=40
  twistLimitDegrees=10
  twistMotion=Locked
 }
 {
  bone1Name=dyn_hair_midle_03
  bone2Name=dyn_hair_midle_02
  swing1LimitDegrees=15
  swing2LimitDegrees=40
  twistLimitDegrees=10
  twistMotion=Locked
 }
 {
  bone1Name=dyn_hair_midle_04
  bone2Name=dyn_hair_midle_03
  swing1LimitDegrees=20
  swing2LimitDegrees=40
  twistLimitDegrees=0
  twistMotion=Locked
 }
 {bone1Name=dyn_hair_right1_02, bone2Name=dyn_hair_right1_01, swing1LimitDegrees=10, swing2LimitDegrees=20, twistLimitDegrees=10}
 {
  bone1Name=dyn_hair_right1_03
  bone2Name=dyn_hair_right1_02
  swing1LimitDegrees=10
  swing2LimitDegrees=20
  twistLimitDegrees=10
  twistMotion=Locked
 }
 {bone1Name=dyn_hair_right2_02, bone2Name=dyn_hair_right2_01, swing1LimitDegrees=20, swing2LimitDegrees=20, twistLimitDegrees=10}
 {bone1Name=mstr_hair_left, bone2Name=dyn_hair_left1_01, swing1LimitDegrees=10, swing2LimitDegrees=20, twistLimitDegrees=5}
 {bone1Name=dyn_hair_left2_01, bone2Name=mstr_hair_left2, swing1LimitDegrees=2, swing2LimitDegrees=2, twistLimitDegrees=15}
 {bone1Name=mstr_hair_left3, bone2Name=dyn_hair_left3_01, swing1LimitDegrees=10, swing2LimitDegrees=15, twistLimitDegrees=5}
 {
  bone1Name=mstr_hair_midle
  bone2Name=dyn_hair_midle_01
  swing1LimitDegrees=10
  swing2LimitDegrees=30
  twistLimitDegrees=5
  twistMotion=Locked
 }
 {bone1Name=mstr_hair_right1, bone2Name=dyn_hair_right1_01, swing1LimitDegrees=15, swing2LimitDegrees=10, twistLimitDegrees=5}
 {bone1Name=mstr_hair_right2, bone2Name=dyn_hair_right2_01, swing1LimitDegrees=10, swing2LimitDegrees=15, twistLimitDegrees=5}
 {
  bone1Name=dyn_hair_midle_02
  bone2Name=dyn_hair_right1_02
  swing1LimitDegrees=45
  swing1Motion=Free
  swing2LimitDegrees=30
  twistLimitDegrees=45
  twistMotion=Free
 }
 {
  bone1Name=dyn_hair_midle_02
  bone2Name=dyn_hair_left1_02
  swing1LimitDegrees=45
  swing1Motion=Free
  swing2LimitDegrees=45
  swing2Motion=Free
  twistLimitDegrees=45
  twistMotion=Free
 }
]
