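# Animation Blueprint for the F_MED_CameraShake face accessory mesh, posing hair and a brow piercing on the player character.
# It copies the attached parent's pose, constrains the piercing to head and facial bones, then chains additive bone-space hair offsets.
# Interpolated state alphas and eased bools drive these corrections for looking, aiming, crouching, falling, skydiving and emotes.
# Output switches between cached 'Rban' and 'Clipping Fix' poses on bRigidBodyEnabled.
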AnimBlueprint /BRCosmetics/Characters/Player/Female/Medium/Bodies/F_MED_CameraShake/Meshes/Parts/F_MED_CameraShake_FaceAcc_AnimBP
target_skeleton=&/F_MED_CameraShake_FaceAcc_Skeleton
graphs [
 AnimGraph {nodes [
   Output Pose(Result=Blend Poses by bool(BlendPose_0=Use cached pose 'Rban'(), BlendPose_1=Use cached pose 'Clipping Fix '(), bActiveValue=Get bRigidBodyEnabled())) {layerGroup=None}
   n1 = "Constraint - brow_piercing (PPPPPP)"!(
    ComponentPose=Local To Component(LocalPose=Copy Pose From Mesh() {bUseAttachedParent=true})
    Alpha=1
    ConstraintWeights_0=0.591
    ConstraintWeights_1:float=0.2818
    ConstraintWeights_2:float=0.0799
    ConstraintWeights_3:float=0.0456
    ConstraintWeights_4:float=0.0015
    ConstraintWeights_5:float=0.0001
   ) -> (Pose) {
    alphaBoolBlend {blendOption=Linear}
    boneToModify {boneName=brow_piercing}
    constraintSetup [
     {targetBone {boneName=head}}
     {targetBone {boneName=FACIAL_R_ForeheadOut}}
     {targetBone {boneName=FACIAL_R_EyesackUpper}}
     {targetBone {boneName=FACIAL_R_ForeheadIn}}
     {targetBone {boneName=FACIAL_R_EyeCornerOuter}}
     {targetBone {boneName=brow_piercing}}
    ]
    constraintWeights [1,1,1,1,1,1]
   }
   n2 = "Transform (Modify) Bone - Bone: mstr_hair"(
    ComponentPose=@n1
    Alpha=Get ClavRoataionMelleFixAlpha() -> (ClavRoataionMelleFixAlpha:float)
    Translation="-1,0,2"
    Rotation="13.497,0.000,0.000"
    Scale="1,1,1"
   ) {
    rotationMode=BMM_Additive
    rotationSpace=BCS_BoneSpace
    scaleMode=BMM_Replace
    scaleSpace=BCS_BoneSpace
    translationMode=BMM_Additive
    translationSpace=BCS_BoneSpace
    alphaBoolBlend {blendInTime=0.05, blendOption=Linear, blendOutTime=0.05}
    alphaScaleBiasClamp {bInterpResult=true, interpSpeedDecreasing=15, interpSpeedIncreasing=15}
    boneToModify {boneName=mstr_hair}
    rotation {pitch=13.497}
    translation {x=-1, z=2}
   }
   n3 = "Transform (Modify) Bone - Bone: mstr_hair_back"(ComponentPose=@n2, Alpha=Get ClavRoataionMelleFixAlpha(), Translation="1,0,2", Rotation="0,0,0", Scale="1,1,1") {
    rotationMode=BMM_Additive
    rotationSpace=BCS_BoneSpace
    scaleMode=BMM_Replace
    scaleSpace=BCS_BoneSpace
    translationMode=BMM_Additive
    translationSpace=BCS_BoneSpace
    alphaBoolBlend {blendInTime=0.05, blendOption=Linear, blendOutTime=0.05}
    alphaScaleBiasClamp {bInterpResult=true, interpSpeedDecreasing=15, interpSpeedIncreasing=15}
    boneToModify {boneName=mstr_hair_back}
    translation {x=1, z=2}
   }
   n4 = "Transform (Modify) Bone - Bone: mstr_hair_left"(ComponentPose=@n3, Alpha=Get ClavRoataionMelleFixAlpha(), Translation="0,0,0", Rotation="43.441,-50.167,-5.474", Scale="1,1,1") {
    rotationMode=BMM_Additive
    rotationSpace=BCS_BoneSpace
    scaleMode=BMM_Replace
    scaleSpace=BCS_BoneSpace
    translationMode=BMM_Additive
    translationSpace=BCS_BoneSpace
    alphaBoolBlend {blendInTime=0.05, blendOption=Linear, blendOutTime=0.05}
    alphaScaleBiasClamp {bInterpResult=true, interpSpeedDecreasing=15, interpSpeedIncreasing=15}
    boneToModify {boneName=mstr_hair_left}
    rotation=<rot 43.441 -50.167 -5.474>
   }
   n5 = "Transform (Modify) Bone - Bone: mstr_hair_midle"(ComponentPose=@n4, Alpha=Get ClavRoataionMelleFixAlpha(), Translation="0,0,0", Rotation="40,0,0", Scale="1,1,1") {
    rotationMode=BMM_Additive
    rotationSpace=BCS_BoneSpace
    scaleMode=BMM_Replace
    scaleSpace=BCS_BoneSpace
    translationMode=BMM_Additive
    translationSpace=BCS_BoneSpace
    alphaBoolBlend {blendInTime=0.05, blendOption=Linear, blendOutTime=0.05}
    alphaScaleBiasClamp {bInterpResult=true, interpSpeedDecreasing=15, interpSpeedIncreasing=15}
    boneToModify {boneName=mstr_hair_midle}
    rotation {pitch=40}
   }
   n6 = "Transform (Modify) Bone - Bone: mstr_hair"(ComponentPose=@n5, Alpha=Get LookUp() -> (LookUp:float), Translation="0,0,0", Rotation="7.354,0.000,0.000", Scale="1,1,1") {
    rotationMode=BMM_Additive
    rotationSpace=BCS_BoneSpace
    scaleMode=BMM_Replace
    scaleSpace=BCS_BoneSpace
    translationMode=BMM_Additive
    translationSpace=BCS_BoneSpace
    alphaBoolBlend {blendInTime=0.2, blendOption=Sinusoidal, blendOutTime=0.4}
    alphaScaleBiasClamp {bInterpResult=true, interpSpeedDecreasing=15, interpSpeedIncreasing=15}
    boneToModify {boneName=mstr_hair}
    rotation {pitch=7.354}
   }
   n7 = "Transform (Modify) Bone - Bone: mstr_hair_back"(ComponentPose=@n6, Alpha=Get LookUp(), Translation="0,0,0.7", Rotation="10,0,0", Scale="1,1,1") {
    rotationMode=BMM_Additive
    rotationSpace=BCS_BoneSpace
    scaleMode=BMM_Replace
    scaleSpace=BCS_BoneSpace
    translationMode=BMM_Additive
    translationSpace=BCS_BoneSpace
    alphaBoolBlend {blendInTime=0.2, blendOption=Sinusoidal, blendOutTime=0.4}
    alphaScaleBiasClamp {bInterpResult=true, interpSpeedDecreasing=15, interpSpeedIncreasing=15}
    boneToModify {boneName=mstr_hair_back}
    rotation {pitch=10}
    translation {z=0.7}
   }
   n8 = "Transform (Modify) Bone - Bone: mstr_hair_right1"(ComponentPose=@n7, Alpha=Get LookUp(), Translation="-2,0,0", Rotation="0,10,0", Scale="1,1,1") {
    rotationMode=BMM_Additive
    rotationSpace=BCS_BoneSpace
    scaleMode=BMM_Replace
    scaleSpace=BCS_BoneSpace
    translationMode=BMM_Additive
    translationSpace=BCS_BoneSpace
    alphaBoolBlend {blendInTime=0.2, blendOption=Sinusoidal, blendOutTime=0.4}
    alphaScaleBiasClamp {bInterpResult=true, interpSpeedDecreasing=15, interpSpeedIncreasing=15}
    boneToModify {boneName=mstr_hair_right1}
    rotation {yaw=10}
    translation {x=-2}
   }
   n9 = "Transform (Modify) Bone - Bone: mstr_hair_midle"(ComponentPose=@n8, Alpha=Get LookUp(), Translation="0,1.127,0", Rotation="6,0,27.755", Scale="1,1,1") {
    rotationMode=BMM_Additive
    rotationSpace=BCS_BoneSpace
    scaleMode=BMM_Replace
    scaleSpace=BCS_BoneSpace
    translationMode=BMM_Additive
    translationSpace=BCS_BoneSpace
    alphaBoolBlend {blendInTime=0.2, blendOption=Sinusoidal, blendOutTime=0.4}
    alphaScaleBiasClamp {bInterpResult=true, interpSpeedDecreasing=15, interpSpeedIncreasing=15}
    boneToModify {boneName=mstr_hair_midle}
    rotation {pitch=6, roll=27.755}
    translation {y=1.127}
   }
   n10 = "Transform (Modify) Bone - Bone: mstr_hair_left2"(ComponentPose=@n9, Alpha=Get LookUp(), Translation="-3,0,0", Rotation="0,-2,0", Scale="1,1,1") {
    rotationMode=BMM_Additive
    rotationSpace=BCS_BoneSpace
    scaleMode=BMM_Replace
    scaleSpace=BCS_BoneSpace
    translationMode=BMM_Additive
    translationSpace=BCS_BoneSpace
    alphaBoolBlend {blendInTime=0.2, blendOption=Sinusoidal, blendOutTime=0.4}
    alphaScaleBiasClamp {bInterpResult=true, interpSpeedDecreasing=15, interpSpeedIncreasing=15}
    boneToModify {boneName=mstr_hair_left2}
    rotation {yaw=-2}
    translation {x=-3}
   }
   n11 = "Transform (Modify) Bone - Bone: mstr_hair"(ComponentPose=@n10, Alpha=Get LookDown() -> (LookDown:float), Translation="0,0,0", Rotation="-10,0,0", Scale="1,1,1") {
    rotationMode=BMM_Additive
    rotationSpace=BCS_BoneSpace
    scaleMode=BMM_Replace
    scaleSpace=BCS_BoneSpace
    translationMode=BMM_Additive
    translationSpace=BCS_BoneSpace
    alphaBoolBlend {blendInTime=0.2, blendOption=Sinusoidal, blendOutTime=0.4}
    alphaScaleBiasClamp {bInterpResult=true, interpSpeedDecreasing=15, interpSpeedIncreasing=15}
    boneToModify {boneName=mstr_hair}
    rotation {pitch=-10}
   }
   n12 = "Transform (Modify) Bone - Bone: mstr_hair_back"(ComponentPose=@n11, Alpha=Get LookDown(), Translation="0,0,0", Rotation="-6.442,0.825,-10.151", Scale="1,1,1") {
    rotationMode=BMM_Additive
    rotationSpace=BCS_BoneSpace
    scaleMode=BMM_Replace
    scaleSpace=BCS_BoneSpace
    translationMode=BMM_Additive
    translationSpace=BCS_BoneSpace
    alphaBoolBlend {blendInTime=0.2, blendOption=Sinusoidal, blendOutTime=0.4}
    alphaScaleBiasClamp {bInterpResult=true, interpSpeedDecreasing=15, interpSpeedIncreasing=15}
    boneToModify {boneName=mstr_hair_back}
    rotation=<rot -6.442 0.825 -10.151>
   }
   n13 = "Transform (Modify) Bone - Bone: mstr_hair_left3"(ComponentPose=@n12, Alpha=Get LookDown(), Translation="0,0,0", Rotation="15,-15,0", Scale="1,1,1") {
    rotationMode=BMM_Additive
    rotationSpace=BCS_BoneSpace
    scaleMode=BMM_Replace
    scaleSpace=BCS_BoneSpace
    translationMode=BMM_Additive
    translationSpace=BCS_BoneSpace
    alphaBoolBlend {blendInTime=0.2, blendOption=Sinusoidal, blendOutTime=0.4}
    alphaScaleBiasClamp {bInterpResult=true, interpSpeedDecreasing=15, interpSpeedIncreasing=15}
    boneToModify {boneName=mstr_hair_left3}
    rotation {pitch=15, yaw=-15}
   }
   n14 = "Transform (Modify) Bone - Bone: mstr_hair_left2"(ComponentPose=@n13, Alpha=Get LookDown(), Translation="-1,1,0", Rotation="-0.000,2.174,-0.000", Scale="1,1,1") {
    rotationMode=BMM_Additive
    rotationSpace=BCS_BoneSpace
    scaleMode=BMM_Replace
    scaleSpace=BCS_BoneSpace
    translationMode=BMM_Additive
    translationSpace=BCS_BoneSpace
    alphaBoolBlend {blendInTime=0.2, blendOption=Sinusoidal, blendOutTime=0.4}
    alphaScaleBiasClamp {bInterpResult=true, interpSpeedDecreasing=15, interpSpeedIncreasing=15}
    boneToModify {boneName=mstr_hair_left2}
    rotation {roll=0, yaw=2.174}
    translation=<-1 1>
   }
   n15 = "Transform (Modify) Bone - Bone: mstr_hair"(
    ComponentPose=@n14
    Alpha=Get LookUpTargeting() -> (LookUpTargeting:float)
    Translation="0,0,0"
    Rotation="-3.779,-27.759,2.126"
    Scale="1,1,1"
   ) {
    rotationMode=BMM_Additive
    rotationSpace=BCS_BoneSpace
    scaleMode=BMM_Replace
    scaleSpace=BCS_BoneSpace
    translationMode=BMM_Additive
    translationSpace=BCS_BoneSpace
    alphaBoolBlend {blendInTime=0.2, blendOption=Sinusoidal, blendOutTime=0.4}
    alphaScaleBiasClamp {bInterpResult=true, interpSpeedDecreasing=15, interpSpeedIncreasing=15}
    boneToModify {boneName=mstr_hair}
    rotation=<rot -3.779 -27.759 2.126>
   }
   n16 = "Transform (Modify) Bone - Bone: mstr_hair_back"(ComponentPose=@n15, Alpha=Get LookUpTargeting(), Translation="-2,-1,3", Rotation="-11.061,1.067,-0.167", Scale="1,1,1") {
    rotationMode=BMM_Additive
    rotationSpace=BCS_BoneSpace
    scaleMode=BMM_Replace
    scaleSpace=BCS_BoneSpace
    translationMode=BMM_Additive
    translationSpace=BCS_BoneSpace
    alphaBoolBlend {blendInTime=0.2, blendOption=Sinusoidal, blendOutTime=0.4}
    alphaScaleBiasClamp {bInterpResult=true, interpSpeedDecreasing=15, interpSpeedIncreasing=15}
    boneToModify {boneName=mstr_hair_back}
    rotation=<rot -11.061 1.067 -0.167>
    translation=<-2 -1 3>
   }
   n17 = "Transform (Modify) Bone - Bone: mstr_hair_right1"(ComponentPose=@n16, Alpha=Get LookUpTargeting(), Translation="-2,0,0", Rotation="0,10,0", Scale="1,1,1") {
    rotationMode=BMM_Additive
    rotationSpace=BCS_BoneSpace
    scaleMode=BMM_Replace
    scaleSpace=BCS_BoneSpace
    translationMode=BMM_Additive
    translationSpace=BCS_BoneSpace
    alphaBoolBlend {blendInTime=0.2, blendOption=Sinusoidal, blendOutTime=0.4}
    alphaScaleBiasClamp {bInterpResult=true, interpSpeedDecreasing=15, interpSpeedIncreasing=15}
    boneToModify {boneName=mstr_hair_right1}
    rotation {yaw=10}
    translation {x=-2}
   }
   n18 = "Transform (Modify) Bone - Bone: mstr_hair_midle"(ComponentPose=@n17, Alpha=Get LookUpTargeting(), Translation="0,0,0", Rotation="45.341,-6.502,-44.671", Scale="1,1,1") {
    rotationMode=BMM_Additive
    rotationSpace=BCS_BoneSpace
    scaleMode=BMM_Replace
    scaleSpace=BCS_BoneSpace
    translationMode=BMM_Additive
    translationSpace=BCS_BoneSpace
    alphaBoolBlend {blendInTime=0.2, blendOption=Sinusoidal, blendOutTime=0.4}
    alphaScaleBiasClamp {bInterpResult=true, interpSpeedDecreasing=15, interpSpeedIncreasing=15}
    boneToModify {boneName=mstr_hair_midle}
    rotation=<rot 45.341 -6.5019 -44.671>
   }
   n19 = "Transform (Modify) Bone - Bone: mstr_hair_back"(
    ComponentPose=@n18
    Alpha=Get LookUpTargeting_Pistol() -> (LookUpTargeting_Pistol:double)
    Translation="0,0,0"
    Rotation="-10,0,0"
    Scale="1,1,1"
   ) {
    rotationMode=BMM_Additive
    rotationSpace=BCS_BoneSpace
    scaleMode=BMM_Replace
    scaleSpace=BCS_BoneSpace
    translationMode=BMM_Additive
    translationSpace=BCS_BoneSpace
    alphaBoolBlend {blendInTime=0.2, blendOption=Sinusoidal, blendOutTime=0.4}
    alphaScaleBiasClamp {bInterpResult=true, interpSpeedDecreasing=15, interpSpeedIncreasing=15}
    boneToModify {boneName=mstr_hair_back}
    rotation {pitch=-10}
   }
   n20 = "Transform (Modify) Bone - Bone: mstr_hair_left2"(ComponentPose=@n19, Alpha=Get LookUpTargeting(), Translation="1.5,-1,0", Rotation="-5.248,5.679,-17.827", Scale="1,1,1") {
    rotationMode=BMM_Additive
    rotationSpace=BCS_BoneSpace
    scaleMode=BMM_Replace
    scaleSpace=BCS_BoneSpace
    translationMode=BMM_Additive
    translationSpace=BCS_BoneSpace
    alphaBoolBlend {blendInTime=0.2, blendOption=Sinusoidal, blendOutTime=0.4}
    alphaScaleBiasClamp {bInterpResult=true, interpSpeedDecreasing=15, interpSpeedIncreasing=15}
    boneToModify {boneName=mstr_hair_left2}
    rotation=<rot -5.2479 5.6791 -17.827>
    translation=<1.5 -1>
   }
   n21 = "Transform (Modify) Bone - Bone: mstr_hair"(
    ComponentPose=@n20
    Alpha=Get LookUpCrouchingState() -> (LookUpCrouchingState:double)
    Translation="0,0,0"
    Rotation="13,-24.083,-30"
    Scale="1,1,1"
   ) {
    rotationMode=BMM_Additive
    rotationSpace=BCS_BoneSpace
    scaleMode=BMM_Replace
    scaleSpace=BCS_BoneSpace
    translationMode=BMM_Additive
    translationSpace=BCS_BoneSpace
    alphaBoolBlend {blendInTime=0.2, blendOption=Sinusoidal, blendOutTime=0.4}
    alphaScaleBiasClamp {bInterpResult=true, interpSpeedDecreasing=15, interpSpeedIncreasing=15}
    boneToModify {boneName=mstr_hair}
    rotation=<rot 13 -24.083 -30>
   }
   n22 = "Transform (Modify) Bone - Bone: mstr_hair_back"(ComponentPose=@n21, Alpha=Get LookUpCrouchingState(), Translation="0,0,3.7", Rotation="3.222,14.454,28.973", Scale="1,1,1") {
    rotationMode=BMM_Additive
    rotationSpace=BCS_BoneSpace
    scaleMode=BMM_Replace
    scaleSpace=BCS_BoneSpace
    translationMode=BMM_Additive
    translationSpace=BCS_BoneSpace
    alphaBoolBlend {blendInTime=0.2, blendOption=Sinusoidal, blendOutTime=0.4}
    alphaScaleBiasClamp {bInterpResult=true, interpSpeedDecreasing=15, interpSpeedIncreasing=15}
    boneToModify {boneName=mstr_hair_back}
    rotation=<rot 3.222 14.454 28.973>
    translation {z=3.7}
   }
   n23 = "Transform (Modify) Bone - Bone: mstr_hair_left2"(ComponentPose=@n22, Alpha=Get LookUpCrouchingState(), Translation="0,0,0", Rotation="0,0,0", Scale="1,1,1") {
    rotationMode=BMM_Additive
    rotationSpace=BCS_BoneSpace
    scaleMode=BMM_Replace
    scaleSpace=BCS_BoneSpace
    translationMode=BMM_Additive
    translationSpace=BCS_BoneSpace
    alphaBoolBlend {blendInTime=0.2, blendOption=Sinusoidal, blendOutTime=0.4}
    alphaScaleBiasClamp {bInterpResult=true, interpSpeedDecreasing=15, interpSpeedIncreasing=15}
    boneToModify {boneName=mstr_hair_left2}
   }
   n24 = "Transform (Modify) Bone - Bone: mstr_hair"(ComponentPose=@n23, Alpha=Get LookUpCrouchingState(), Translation="0,0,0", Rotation="0,0,0", Scale="1,1,1") {
    rotationMode=BMM_Additive
    rotationSpace=BCS_BoneSpace
    scaleMode=BMM_Replace
    scaleSpace=BCS_BoneSpace
    translationMode=BMM_Additive
    translationSpace=BCS_BoneSpace
    alphaBoolBlend {blendInTime=0.2, blendOption=Sinusoidal, blendOutTime=0.4}
    alphaScaleBiasClamp {bInterpResult=true, interpSpeedDecreasing=15, interpSpeedIncreasing=15}
    boneToModify {boneName=mstr_hair}
   }
   n25 = "Transform (Modify) Bone - Bone: mstr_hair_left2"(ComponentPose=@n24, Alpha=Get LookUpCrouchingState(), Translation="-3,0,2", Rotation="-0.017,-15,0.053", Scale="1,1,1") {
    rotationMode=BMM_Additive
    rotationSpace=BCS_BoneSpace
    scaleMode=BMM_Replace
    scaleSpace=BCS_BoneSpace
    translationMode=BMM_Additive
    translationSpace=BCS_BoneSpace
    alphaBoolBlend {blendInTime=0.2, blendOption=Sinusoidal, blendOutTime=0.4}
    alphaScaleBiasClamp {bInterpResult=true, interpSpeedDecreasing=15, interpSpeedIncreasing=15}
    boneToModify {boneName=mstr_hair_left2}
    rotation=<rot -0.017 -15 0.053>
    translation {x=-3, z=2}
   }
   n26 = "Transform (Modify) Bone - Bone: mstr_hair_left3"(ComponentPose=@n25, Alpha=Get LookUpCrouchingState(), Translation="-2,-1,0", Rotation="0,0,0", Scale="1,1,1") {
    rotationMode=BMM_Additive
    rotationSpace=BCS_BoneSpace
    scaleMode=BMM_Replace
    scaleSpace=BCS_BoneSpace
    translationMode=BMM_Additive
    translationSpace=BCS_BoneSpace
    alphaBoolBlend {blendInTime=0.2, blendOption=Sinusoidal, blendOutTime=0.4}
    alphaScaleBiasClamp {bInterpResult=true, interpSpeedDecreasing=15, interpSpeedIncreasing=15}
    boneToModify {boneName=mstr_hair_left3}
    translation=<-2 -1>
   }
   n27 = "Transform (Modify) Bone - Bone: mstr_hair_right2"(ComponentPose=@n26, Alpha=Get LookUpCrouchingState(), Translation="0,0,0", Rotation="0,0,0", Scale="1,1,1") {
    rotationMode=BMM_Additive
    rotationSpace=BCS_BoneSpace
    scaleMode=BMM_Replace
    scaleSpace=BCS_BoneSpace
    translationMode=BMM_Additive
    translationSpace=BCS_BoneSpace
    alphaBoolBlend {blendInTime=0.2, blendOption=Sinusoidal, blendOutTime=0.4}
    alphaScaleBiasClamp {bInterpResult=true, interpSpeedDecreasing=15, interpSpeedIncreasing=15}
    boneToModify {boneName=mstr_hair_right2}
   }
   n28 = "Transform (Modify) Bone - Bone: mstr_hair"(
    ComponentPose=@n27
    Alpha=Get LookUpCorunchingAndTargeting() -> (LookUpCorunchingAndTargeting:double)
    Translation="1,0,0"
    Rotation="10.912,-0.717,-1.329"
    Scale="1,1,1"
   ) {
    rotationMode=BMM_Additive
    rotationSpace=BCS_BoneSpace
    scaleMode=BMM_Replace
    scaleSpace=BCS_BoneSpace
    translationMode=BMM_Additive
    translationSpace=BCS_BoneSpace
    alphaBoolBlend {blendInTime=0.2, blendOption=Sinusoidal, blendOutTime=0.4}
    alphaScaleBiasClamp {bInterpResult=true, interpSpeedDecreasing=15, interpSpeedIncreasing=15}
    boneToModify {boneName=mstr_hair}
    rotation=<rot 10.9119 -0.717 -1.329>
    translation {x=1}
   }
   n29 = "Transform (Modify) Bone - Bone: mstr_hair_left2"(ComponentPose=@n28, Alpha=Get LookUpCorunchingAndTargeting(), Translation="2.794,-1.941,0.499", Rotation="-4.144,2.989,4.378", Scale="1,1,1") {
    rotationMode=BMM_Additive
    rotationSpace=BCS_BoneSpace
    scaleMode=BMM_Replace
    scaleSpace=BCS_BoneSpace
    translationMode=BMM_Additive
    translationSpace=BCS_BoneSpace
    alphaBoolBlend {blendInTime=0.2, blendOption=Sinusoidal, blendOutTime=0.4}
    alphaScaleBiasClamp {bInterpResult=true, interpSpeedDecreasing=15, interpSpeedIncreasing=15}
    boneToModify {boneName=mstr_hair_left2}
    rotation=<rot -4.144 2.989 4.378>
    translation=<2.794 -1.941 0.499>
   }
   n30 = "Transform (Modify) Bone - Bone: mstr_hair_right1"(ComponentPose=@n29, Alpha=Get LookUpCorunchingAndTargeting(), Translation="0,0,0", Rotation="15.000,-0.044,0.073", Scale="1,1,1") {
    rotationMode=BMM_Additive
    rotationSpace=BCS_BoneSpace
    scaleMode=BMM_Replace
    scaleSpace=BCS_BoneSpace
    translationMode=BMM_Additive
    translationSpace=BCS_BoneSpace
    alphaBoolBlend {blendInTime=0.2, blendOption=Sinusoidal, blendOutTime=0.4}
    alphaScaleBiasClamp {bInterpResult=true, interpSpeedDecreasing=15, interpSpeedIncreasing=15}
    boneToModify {boneName=mstr_hair_right1}
    rotation=<rot 14.9999 -0.044 0.073>
   }
   n31 = "Transform (Modify) Bone - Bone: mstr_hair"(
    ComponentPose=@n30
    Alpha=Get LookUpCrouchingAndMoving() -> (LookUpCrouchingAndMoving:double)
    Translation="-2,0,5"
    Rotation="50.286,-18.289,-10.948"
    Scale="1,1,1"
   ) {
    rotationMode=BMM_Additive
    rotationSpace=BCS_BoneSpace
    scaleMode=BMM_Replace
    scaleSpace=BCS_BoneSpace
    translationMode=BMM_Additive
    translationSpace=BCS_BoneSpace
    alphaBoolBlend {blendInTime=0.2, blendOption=Sinusoidal, blendOutTime=0.4}
    alphaScaleBiasClamp {bInterpResult=true, interpSpeedDecreasing=15, interpSpeedIncreasing=15}
    boneToModify {boneName=mstr_hair}
    rotation=<rot 50.2859 -18.289 -10.948>
    translation {x=-2, z=5}
   }
   n32 = "Transform (Modify) Bone - Bone: mstr_hair"(
    ComponentPose=@n31
    Alpha=Get LookUpCrouchingAndMoving_Targeting() -> (LookUpCrouchingAndMoving_Targeting:double)
    Translation="0,0,0"
    Rotation="15,0,0"
    Scale="1,1,1"
   ) {
    rotationMode=BMM_Additive
    rotationSpace=BCS_BoneSpace
    scaleMode=BMM_Replace
    scaleSpace=BCS_BoneSpace
    translationMode=BMM_Additive
    translationSpace=BCS_BoneSpace
    alphaBoolBlend {blendInTime=0.2, blendOption=Sinusoidal, blendOutTime=0.4}
    alphaScaleBiasClamp {bInterpResult=true, interpSpeedDecreasing=15, interpSpeedIncreasing=15}
    boneToModify {boneName=mstr_hair}
    rotation {pitch=15}
   }
   n33 = Get LookUpCrouchingAndMoving_Targeting()
   n34 = "Transform (Modify) Bone - Bone: mstr_hair_back"(ComponentPose=@n32, Alpha=@n33, Translation="0,0,0", Rotation="0,0,0", Scale="1,1,1") {
    rotationMode=BMM_Additive
    rotationSpace=BCS_BoneSpace
    scaleMode=BMM_Replace
    scaleSpace=BCS_BoneSpace
    translationMode=BMM_Additive
    translationSpace=BCS_BoneSpace
    alphaBoolBlend {blendInTime=0.2, blendOption=Sinusoidal, blendOutTime=0.4}
    alphaScaleBiasClamp {bInterpResult=true, interpSpeedDecreasing=15, interpSpeedIncreasing=15}
    boneToModify {boneName=mstr_hair_back}
   }
   n35 = "Transform (Modify) Bone - Bone: mstr_hair_left2"(ComponentPose=@n34, Alpha=@n33, Translation="0,0,0", Rotation="1.740,14.458,-4.924", Scale="1,1,1") {
    rotationMode=BMM_Additive
    rotationSpace=BCS_BoneSpace
    scaleMode=BMM_Replace
    scaleSpace=BCS_BoneSpace
    translationMode=BMM_Additive
    translationSpace=BCS_BoneSpace
    alphaBoolBlend {blendInTime=0.2, blendOption=Sinusoidal, blendOutTime=0.4}
    alphaScaleBiasClamp {bInterpResult=true, interpSpeedDecreasing=15, interpSpeedIncreasing=15}
    boneToModify {boneName=mstr_hair_left2}
    rotation=<rot 1.74 14.458 -4.924>
   }
   n36 = "Transform (Modify) Bone - Bone: mstr_hair_left3"(ComponentPose=@n35, Alpha=Get LookUpCrouchingAndMoving_Targeting(), Translation="-1.922,3.604,0.889", Rotation="0,0,0", Scale="1,1,1") {
    rotationMode=BMM_Additive
    rotationSpace=BCS_BoneSpace
    scaleMode=BMM_Replace
    scaleSpace=BCS_BoneSpace
    translationMode=BMM_Additive
    translationSpace=BCS_BoneSpace
    alphaBoolBlend {blendInTime=0.2, blendOption=Sinusoidal, blendOutTime=0.4}
    alphaScaleBiasClamp {bInterpResult=true, interpSpeedDecreasing=15, interpSpeedIncreasing=15}
    boneToModify {boneName=mstr_hair_left3}
    translation=<-1.922 3.604 0.889>
   }
   n37 = "Transform (Modify) Bone - Bone: mstr_hair_right2"(ComponentPose=@n36, Alpha=@n33, Translation="0.071,-0.364,0.233", Rotation="0,0,0", Scale="1,1,1") {
    rotationMode=BMM_Additive
    rotationSpace=BCS_BoneSpace
    scaleMode=BMM_Replace
    scaleSpace=BCS_BoneSpace
    translationMode=BMM_Additive
    translationSpace=BCS_BoneSpace
    alphaBoolBlend {blendInTime=0.2, blendOption=Sinusoidal, blendOutTime=0.4}
    alphaScaleBiasClamp {bInterpResult=true, interpSpeedDecreasing=15, interpSpeedIncreasing=15}
    boneToModify {boneName=mstr_hair_right2}
    translation=<0.071 -0.364 0.233>
   }
   n38 = Get MiddleCrouchingAndMoving() -> (MiddleCrouchingAndMoving:double)
   n39 = "Transform (Modify) Bone - Bone: mstr_hair"(ComponentPose=@n37, Alpha=@n38, Translation="0,0,0", Rotation="13.211,-0.666,-2.115", Scale="1,1,1") {
    rotationMode=BMM_Additive
    rotationSpace=BCS_BoneSpace
    scaleMode=BMM_Replace
    scaleSpace=BCS_BoneSpace
    translationMode=BMM_Additive
    translationSpace=BCS_BoneSpace
    alphaBoolBlend {blendInTime=0.2, blendOption=Sinusoidal, blendOutTime=0.4}
    alphaScaleBiasClamp {bInterpResult=true, interpSpeedDecreasing=15, interpSpeedIncreasing=15}
    boneToModify {boneName=mstr_hair}
    rotation=<rot 13.211 -0.666 -2.115>
   }
   n40 = "Transform (Modify) Bone - Bone: mstr_hair_back"(ComponentPose=@n39, Alpha=@n38, Translation="0,0,4", Rotation="-13.818,0.035,-0.055", Scale="1,1,1") {
    rotationMode=BMM_Additive
    rotationSpace=BCS_BoneSpace
    scaleMode=BMM_Replace
    scaleSpace=BCS_BoneSpace
    translationMode=BMM_Additive
    translationSpace=BCS_BoneSpace
    alphaBoolBlend {blendInTime=0.2, blendOption=Sinusoidal, blendOutTime=0.4}
    alphaScaleBiasClamp {bInterpResult=true, interpSpeedDecreasing=15, interpSpeedIncreasing=15}
    boneToModify {boneName=mstr_hair_back}
    rotation=<rot -13.818 0.035 -0.055>
    translation {z=4}
   }
   n41 = "Transform (Modify) Bone - Bone: mstr_hair_left"(ComponentPose=@n40, Alpha=@n38, Translation="0,0,0", Rotation="18.724,-4.517,15.916", Scale="1,1,1") {
    rotationMode=BMM_Additive
    rotationSpace=BCS_BoneSpace
    scaleMode=BMM_Replace
    scaleSpace=BCS_BoneSpace
    translationMode=BMM_Additive
    translationSpace=BCS_BoneSpace
    alphaBoolBlend {blendInTime=0.2, blendOption=Sinusoidal, blendOutTime=0.4}
    alphaScaleBiasClamp {bInterpResult=true, interpSpeedDecreasing=15, interpSpeedIncreasing=15}
    boneToModify {boneName=mstr_hair_left}
    rotation=<rot 18.724 -4.5169 15.916>
   }
   n42 = "Transform (Modify) Bone - Bone: mstr_hair"(ComponentPose=@n41, Alpha=Get LookDownCrouching() -> (LookDownCrouching:double), Translation="-1,0,0", Rotation="-10,0,0.11", Scale="1,1,1") {
    rotationMode=BMM_Additive
    rotationSpace=BCS_BoneSpace
    scaleMode=BMM_Replace
    scaleSpace=BCS_BoneSpace
    translationMode=BMM_Additive
    translationSpace=BCS_BoneSpace
    alphaBoolBlend {blendInTime=0.2, blendOption=Sinusoidal, blendOutTime=0.4}
    alphaScaleBiasClamp {bInterpResult=true, interpSpeedDecreasing=15, interpSpeedIncreasing=15}
    boneToModify {boneName=mstr_hair}
    rotation {pitch=-10, roll=0.11}
    translation {x=-1}
   }
   n43 = "Transform (Modify) Bone - Bone: mstr_hair_midle"(ComponentPose=@n42, Alpha=Get LookDownCrouching(), Translation="0,0,0", Rotation="-0.006,27.264,0.028", Scale="1,1,1") {
    rotationMode=BMM_Additive
    rotationSpace=BCS_BoneSpace
    scaleMode=BMM_Replace
    scaleSpace=BCS_BoneSpace
    translationMode=BMM_Additive
    translationSpace=BCS_BoneSpace
    alphaBoolBlend {blendInTime=0.2, blendOption=Sinusoidal, blendOutTime=0.4}
    alphaScaleBiasClamp {bInterpResult=true, interpSpeedDecreasing=15, interpSpeedIncreasing=15}
    boneToModify {boneName=mstr_hair_midle}
    rotation=<rot -0.006 27.264 0.028>
   }
   n44 = "Transform (Modify) Bone - Bone: mstr_hair_back"(
    ComponentPose=@n43
    Alpha=Get LookDownCorunchingAndTargeting() -> (LookDownCorunchingAndTargeting:double)
    Translation="0,0,0"
    Rotation="5.052,-0.358,18.373"
    Scale="1,1,1"
   ) {
    rotationMode=BMM_Additive
    rotationSpace=BCS_BoneSpace
    scaleMode=BMM_Replace
    scaleSpace=BCS_BoneSpace
    translationMode=BMM_Additive
    translationSpace=BCS_BoneSpace
    alphaBoolBlend {blendInTime=0.2, blendOption=Sinusoidal, blendOutTime=0.4}
    alphaScaleBiasClamp {bInterpResult=true, interpSpeedDecreasing=15, interpSpeedIncreasing=15}
    boneToModify {boneName=mstr_hair_back}
    rotation=<rot 5.052 -0.358 18.373>
   }
   n45 = "Transform (Modify) Bone - Bone: mstr_hair_midle"(ComponentPose=@n44, Alpha=Get LookDownCorunchingAndTargeting(), Translation="0,0,0", Rotation="0.006,-27.264,-0.028", Scale="1,1,1") {
    rotationMode=BMM_Additive
    rotationSpace=BCS_BoneSpace
    scaleMode=BMM_Replace
    scaleSpace=BCS_BoneSpace
    translationMode=BMM_Additive
    translationSpace=BCS_BoneSpace
    alphaBoolBlend {blendInTime=0.2, blendOption=Sinusoidal, blendOutTime=0.4}
    alphaScaleBiasClamp {bInterpResult=true, interpSpeedDecreasing=15, interpSpeedIncreasing=15}
    boneToModify {boneName=mstr_hair_midle}
    rotation=<rot 0.006 -27.264 -0.028>
   }
   n46 = "Transform (Modify) Bone - Bone: mstr_hair_right1"(ComponentPose=@n45, Alpha=Get LookDownCorunchingAndTargeting(), Translation="0.984,0.004,-3.141", Rotation="0,0,0", Scale="1,1,1") {
    rotationMode=BMM_Additive
    rotationSpace=BCS_BoneSpace
    scaleMode=BMM_Replace
    scaleSpace=BCS_BoneSpace
    translationMode=BMM_Additive
    translationSpace=BCS_BoneSpace
    alphaBoolBlend {blendInTime=0.2, blendOption=Sinusoidal, blendOutTime=0.4}
    alphaScaleBiasClamp {bInterpResult=true, interpSpeedDecreasing=15, interpSpeedIncreasing=15}
    boneToModify {boneName=mstr_hair_right1}
    translation=<0.984 0.004 -3.141>
   }
   n47 = "Transform (Modify) Bone - Bone: mstr_hair"(ComponentPose=@n46, Alpha=1, Translation="0,0,0", Rotation="0,0,0", Scale="1,1,1") {
    rotationMode=BMM_Additive
    rotationSpace=BCS_BoneSpace
    scaleMode=BMM_Replace
    scaleSpace=BCS_BoneSpace
    translationMode=BMM_Additive
    translationSpace=BCS_BoneSpace
    alphaBoolBlend {blendInTime=0.2, blendOption=Sinusoidal, blendOutTime=0.4}
    alphaScaleBiasClamp {bInterpResult=true, interpSpeedDecreasing=15, interpSpeedIncreasing=15}
    boneToModify {boneName=mstr_hair}
   }
   n48 = "Transform (Modify) Bone - Bone: mstr_hair"(
    ComponentPose=@n47
    Alpha=Get LookDownTargeting() -> (LookDownTargeting:double)
    Translation="0,0,0"
    Rotation="21.748,0.051,0.033"
    Scale="1,1,1"
   ) {
    rotationMode=BMM_Additive
    rotationSpace=BCS_BoneSpace
    scaleMode=BMM_Replace
    scaleSpace=BCS_BoneSpace
    translationMode=BMM_Additive
    translationSpace=BCS_BoneSpace
    alphaBoolBlend {blendInTime=0.2, blendOption=Sinusoidal, blendOutTime=0.4}
    alphaScaleBiasClamp {bInterpResult=true, interpSpeedDecreasing=15, interpSpeedIncreasing=15}
    boneToModify {boneName=mstr_hair}
    rotation=<rot 21.7479 0.0511 0.0329>
   }
   n49 = "Transform (Modify) Bone - Bone: mstr_hair_back"(ComponentPose=@n48, Alpha=Get LookDownTargeting(), Translation="0,0,0", Rotation="0,0,0", Scale="1,1,1") {
    rotationMode=BMM_Additive
    rotationSpace=BCS_BoneSpace
    scaleMode=BMM_Replace
    scaleSpace=BCS_BoneSpace
    translationMode=BMM_Additive
    translationSpace=BCS_BoneSpace
    alphaBoolBlend {blendInTime=0.2, blendOption=Sinusoidal, blendOutTime=0.4}
    alphaScaleBiasClamp {bInterpResult=true, interpSpeedDecreasing=15, interpSpeedIncreasing=15}
    boneToModify {boneName=mstr_hair_back}
   }
   n50 = "Transform (Modify) Bone - Bone: mstr_hair_right1"(ComponentPose=@n49, Alpha=Get LookDownTargeting(), Translation="-2,0,0", Rotation="0,10,0", Scale="1,1,1") {
    rotationMode=BMM_Additive
    rotationSpace=BCS_BoneSpace
    scaleMode=BMM_Replace
    scaleSpace=BCS_BoneSpace
    translationMode=BMM_Additive
    translationSpace=BCS_BoneSpace
    alphaBoolBlend {blendInTime=0.2, blendOption=Sinusoidal, blendOutTime=0.4}
    alphaScaleBiasClamp {bInterpResult=true, interpSpeedDecreasing=15, interpSpeedIncreasing=15}
    boneToModify {boneName=mstr_hair_right1}
    rotation {yaw=10}
    translation {x=-2}
   }
   n51 = "Transform (Modify) Bone - Bone: mstr_hair_midle"!(
    ComponentPose=@n50
    bAlphaBoolEnabled:bool=Get IsFallnigAfterJumpWithoutBackpack() -> (IsFallnigAfterJumpWithoutBackpack:bool)
    Translation="0.470,-0.244,1.715"
    Rotation="40,0,0"
    Scale="1,1,1"
   ) -> (Pose) {
    alphaInputType=Bool
    rotationMode=BMM_Additive
    rotationSpace=BCS_BoneSpace
    scaleMode=BMM_Replace
    scaleSpace=BCS_BoneSpace
    translationMode=BMM_Additive
    translationSpace=BCS_BoneSpace
    alphaBoolBlend {blendInTime=0.2, blendOption=Sinusoidal, blendOutTime=0.5}
    alphaScaleBiasClamp {bInterpResult=true, interpSpeedDecreasing=15, interpSpeedIncreasing=15}
    boneToModify {boneName=mstr_hair_midle}
    rotation {pitch=40}
    translation=<0.47 -0.244 1.715>
   }
   n52 = "Transform (Modify) Bone - Bone: mstr_hair_right1"!(
    ComponentPose=@n51
    bAlphaBoolEnabled:bool=Get IsFallnigAfterJumpWithoutBackpack()
    Translation="-1.042,0,0"
    Rotation="0,0,0"
    Scale="1,1,1"
   ) -> (Pose) {
    alphaInputType=Bool
    rotationMode=BMM_Additive
    rotationSpace=BCS_BoneSpace
    scaleMode=BMM_Replace
    scaleSpace=BCS_BoneSpace
    translationMode=BMM_Additive
    translationSpace=BCS_BoneSpace
    alphaBoolBlend {blendInTime=0.3, blendOption=Sinusoidal, blendOutTime=0.4}
    alphaScaleBiasClamp {bInterpResult=true, interpSpeedDecreasing=15, interpSpeedIncreasing=15}
    boneToModify {boneName=mstr_hair_right1}
    translation {x=-1.042}
   }
   n53 = "Transform (Modify) Bone - Bone: mstr_hair_left"!(
    ComponentPose=@n52
    bAlphaBoolEnabled:bool=Get IsFallnigAfterJumpWithoutBackpack()
    Translation="-1.042,-0.981,0.000"
    Rotation="60,-20,-0.076"
    Scale="1,1,1"
   ) -> (Pose) {
    alphaInputType=Bool
    rotationMode=BMM_Additive
    rotationSpace=BCS_BoneSpace
    scaleMode=BMM_Replace
    scaleSpace=BCS_BoneSpace
    translationMode=BMM_Additive
    translationSpace=BCS_BoneSpace
    alphaBoolBlend {blendInTime=0.3, blendOption=Sinusoidal, blendOutTime=0.4}
    alphaScaleBiasClamp {bInterpResult=true, interpSpeedDecreasing=15, interpSpeedIncreasing=15}
    boneToModify {boneName=mstr_hair_left}
    rotation=<rot 60 -20 -0.076>
    translation=<-1.042 -0.981>
   }
   n54 = "Transform (Modify) Bone - Bone: mstr_hair"!(
    ComponentPose=@n53
    bAlphaBoolEnabled:bool=Get IsFallnigAfterJumpWithBackpack() -> (IsFallnigAfterJumpWithBackpack:bool)
    Translation="0,0,0"
    Rotation="10,0,0"
    Scale="1,1,1"
   ) -> (Pose) {
    alphaInputType=Bool
    rotationMode=BMM_Additive
    rotationSpace=BCS_BoneSpace
    scaleMode=BMM_Replace
    scaleSpace=BCS_BoneSpace
    translationMode=BMM_Additive
    translationSpace=BCS_BoneSpace
    alphaBoolBlend {blendInTime=0.2, blendOption=Sinusoidal, blendOutTime=0.4}
    alphaScaleBiasClamp {bInterpResult=true, interpSpeedDecreasing=15, interpSpeedIncreasing=15}
    boneToModify {boneName=mstr_hair}
    rotation {pitch=10}
   }
   n55 = "Transform (Modify) Bone - Bone: mstr_hair_right1"(
    ComponentPose=@n54
    bAlphaBoolEnabled=Get IsFallnigAfterJumpWithBackpack()
    Translation="-0.8,0,0"
    Rotation="-15.042,-11.913,4.089"
    Scale="1,1,1"
   ) {
    alphaInputType=Bool
    rotationMode=BMM_Additive
    rotationSpace=BCS_BoneSpace
    scaleMode=BMM_Replace
    scaleSpace=BCS_BoneSpace
    translationMode=BMM_Additive
    translationSpace=BCS_BoneSpace
    alphaBoolBlend {blendInTime=0.3, blendOption=Sinusoidal, blendOutTime=0.4}
    alphaScaleBiasClamp {bInterpResult=true, interpSpeedDecreasing=15, interpSpeedIncreasing=15}
    boneToModify {boneName=mstr_hair_right1}
    rotation=<rot -15.042 -11.913 4.089>
    translation {x=-0.8}
   }
   n56 = "Transform (Modify) Bone - Bone: mstr_hair_back"!(ComponentPose=@n55, bAlphaBoolEnabled:bool=Get IsFallnigAfterJumpWithBackpack(), Translation="-1,0,0", Rotation="10,0,0", Scale="1,1,1") -> (Pose) {
    alphaInputType=Bool
    rotationMode=BMM_Additive
    rotationSpace=BCS_BoneSpace
    scaleMode=BMM_Replace
    scaleSpace=BCS_BoneSpace
    translationMode=BMM_Additive
    translationSpace=BCS_BoneSpace
    alphaBoolBlend {blendInTime=0.3, blendOption=Sinusoidal, blendOutTime=0.4}
    alphaScaleBiasClamp {bInterpResult=true, interpSpeedDecreasing=15, interpSpeedIncreasing=15}
    boneToModify {boneName=mstr_hair_back}
    rotation {pitch=10}
    translation {x=-1}
   }
   n57 = "Transform (Modify) Bone - Bone: mstr_hair"(ComponentPose=@n56, bAlphaBoolEnabled=Get bIsSkyDiving(), Translation="0,0,0", Rotation="-3,0,0", Scale="1,1,1") {
    alphaInputType=Bool
    rotationMode=BMM_Additive
    rotationSpace=BCS_BoneSpace
    scaleMode=BMM_Replace
    scaleSpace=BCS_BoneSpace
    translationMode=BMM_Additive
    translationSpace=BCS_BoneSpace
    alphaBoolBlend {blendInTime=0.2, blendOption=Sinusoidal, blendOutTime=0.4}
    alphaScaleBiasClamp {bInterpResult=true, interpSpeedDecreasing=15, interpSpeedIncreasing=15}
    boneToModify {boneName=mstr_hair}
    rotation {pitch=-3}
   }
   n58 = "Transform (Modify) Bone - Bone: mstr_hair_left3"!(
    ComponentPose=@n57
    bAlphaBoolEnabled:bool=Get SkydiwingButNotParachute() -> (SkydiwingButNotParachute:bool)
    Translation="-0.576,-6.974,5.263"
    Rotation="-30.637,-19.277,-14.226"
    Scale="1,1,1"
   ) -> (Pose) {
    alphaInputType=Bool
    rotationMode=BMM_Additive
    rotationSpace=BCS_BoneSpace
    scaleMode=BMM_Replace
    scaleSpace=BCS_BoneSpace
    translationMode=BMM_Additive
    translationSpace=BCS_BoneSpace
    alphaBoolBlend {blendInTime=0.2, blendOption=Sinusoidal, blendOutTime=0.4}
    alphaScaleBiasClamp {bInterpResult=true, interpSpeedDecreasing=15, interpSpeedIncreasing=15}
    boneToModify {boneName=mstr_hair_left3}
    rotation=<rot -30.637 -19.2769 -14.226>
    translation=<-0.576 -6.974 5.263>
   }
   n59 = "Transform (Modify) Bone - Bone: mstr_hair_back"(ComponentPose=@n58, bAlphaBoolEnabled=Get bIsSkyDiving(), Translation="-1,0,2", Rotation="0,0,0", Scale="1,1,1") {
    alphaInputType=Bool
    rotationMode=BMM_Additive
    rotationSpace=BCS_BoneSpace
    scaleMode=BMM_Replace
    scaleSpace=BCS_BoneSpace
    translationMode=BMM_Additive
    translationSpace=BCS_BoneSpace
    alphaBoolBlend {blendInTime=0.2, blendOption=Sinusoidal, blendOutTime=0.4}
    alphaScaleBiasClamp {bInterpResult=true, interpSpeedDecreasing=15, interpSpeedIncreasing=15}
    boneToModify {boneName=mstr_hair_back}
    translation {x=-1, z=2}
   }
   n60 = "Transform (Modify) Bone - Bone: mstr_hair_left2"!(
    ComponentPose=@n59
    bAlphaBoolEnabled:bool=Get bIsSkyDiving()
    Translation="-1.140,-0.139,-0.117"
    Rotation="-13.837,-8.883,2.139"
    Scale="1,1,1"
   ) -> (Pose) {
    alphaInputType=Bool
    rotationMode=BMM_Additive
    rotationSpace=BCS_BoneSpace
    scaleMode=BMM_Replace
    scaleSpace=BCS_BoneSpace
    translationMode=BMM_Additive
    translationSpace=BCS_BoneSpace
    alphaBoolBlend {blendInTime=0.2, blendOption=Sinusoidal, blendOutTime=0.4}
    alphaScaleBiasClamp {bInterpResult=true, interpSpeedDecreasing=15, interpSpeedIncreasing=15}
    boneToModify {boneName=mstr_hair_left2}
    rotation=<rot -13.837 -8.883 2.139>
    translation=<-1.1401 -0.139 -0.1172>
   }
   n61 = "Transform (Modify) Bone - Bone: mstr_hair"(
    ComponentPose=@n60
    bAlphaBoolEnabled=Get SkydiwingWithBackpac() -> (SkydiwingWithBackpac:bool)
    Translation="0,0,0"
    Rotation="-5,0.002,0.013"
    Scale="1,1,1"
   ) {
    alphaInputType=Bool
    rotationMode=BMM_Additive
    rotationSpace=BCS_BoneSpace
    scaleMode=BMM_Replace
    scaleSpace=BCS_BoneSpace
    translationMode=BMM_Additive
    translationSpace=BCS_BoneSpace
    alphaBoolBlend {blendInTime=0.2, blendOption=Sinusoidal, blendOutTime=0.4}
    alphaScaleBiasClamp {bInterpResult=true, interpSpeedDecreasing=15, interpSpeedIncreasing=15}
    boneToModify {boneName=mstr_hair}
    rotation=<rot -5 0.002 0.013>
   }
   n62 = "Transform (Modify) Bone - Bone: mstr_hair_back"(ComponentPose=@n61, bAlphaBoolEnabled=Get SkydiwingWithBackpac(), Translation="0,0,1", Rotation="0,0,0", Scale="1,1,1") {
    alphaInputType=Bool
    rotationMode=BMM_Additive
    rotationSpace=BCS_BoneSpace
    scaleMode=BMM_Replace
    scaleSpace=BCS_BoneSpace
    translationMode=BMM_Additive
    translationSpace=BCS_BoneSpace
    alphaBoolBlend {blendInTime=0.2, blendOption=Sinusoidal, blendOutTime=0.4}
    alphaScaleBiasClamp {bInterpResult=true, interpSpeedDecreasing=15, interpSpeedIncreasing=15}
    boneToModify {boneName=mstr_hair_back}
    translation {z=1}
   }
   n63 = "Transform (Modify) Bone - Bone: mstr_hair"(ComponentPose=@n62, bAlphaBoolEnabled=Get bIsPlayingEmote(), Translation="0,0,0", Rotation="5,0,0", Scale="1,1,1") {
    alphaInputType=Bool
    rotationMode=BMM_Additive
    rotationSpace=BCS_BoneSpace
    scaleMode=BMM_Replace
    scaleSpace=BCS_BoneSpace
    translationMode=BMM_Additive
    translationSpace=BCS_BoneSpace
    alphaBoolBlend {blendInTime=0.2, blendOption=Sinusoidal, blendOutTime=0.4}
    alphaScaleBiasClamp {bInterpResult=true, interpSpeedDecreasing=15, interpSpeedIncreasing=15}
    boneToModify {boneName=mstr_hair}
    rotation {pitch=5}
   }
   n64 = "Transform (Modify) Bone - Bone: mstr_hair_back"(ComponentPose=@n63, bAlphaBoolEnabled=Get bIsPlayingEmote(), Translation="-1,0,3", Rotation="0,0,0", Scale="1,1,1") {
    alphaInputType=Bool
    rotationMode=BMM_Additive
    rotationSpace=BCS_BoneSpace
    scaleMode=BMM_Replace
    scaleSpace=BCS_BoneSpace
    translationMode=BMM_Additive
    translationSpace=BCS_BoneSpace
    alphaBoolBlend {blendInTime=0.1, blendOption=Sinusoidal, blendOutTime=0.8}
    alphaScaleBiasClamp {bInterpResult=true, interpSpeedDecreasing=15, interpSpeedIncreasing=15}
    boneToModify {boneName=mstr_hair_back}
    translation {x=-1, z=3}
   }
   n65 = "Transform (Modify) Bone - Bone: mstr_hair_left3"(ComponentPose=@n64, bAlphaBoolEnabled=Get bIsInFrontEnd(), Translation="-0.18,1,-0.102", Rotation="5.572,-22.115,4.083", Scale="1,1,1") {
    alphaInputType=Bool
    rotationMode=BMM_Additive
    rotationSpace=BCS_BoneSpace
    scaleMode=BMM_Replace
    scaleSpace=BCS_BoneSpace
    translationMode=BMM_Additive
    translationSpace=BCS_BoneSpace
    alphaBoolBlend {blendInTime=0.2, blendOption=Sinusoidal, blendOutTime=0.4}
    alphaScaleBiasClamp {bInterpResult=true, interpSpeedDecreasing=15, interpSpeedIncreasing=15}
    boneToModify {boneName=mstr_hair_left3}
    rotation=<rot 5.572 -22.1149 4.083>
    translation=<-0.18 1 -0.102>
   }
   n66 = RigidBody!(
    ComponentPose=@n65
    Alpha
    OverrideWorldGravity:Vector=Get Gravity Override()
    ComponentLinearAccScale:Vector=Get Component Linear Acc Scale() -> (Component Linear Acc Scale:Vector)
    ComponentLinearVelScale:Vector=Get Component Linear Vel Scale() -> (Component Linear Vel Scale:Vector)
   ) -> (Pose) {
    bClampLinearTranslationLimitToRefPose=true
    bOverrideWorldGravity=true
    overridePhysicsAsset=&/F_MED_CameraShake_FaceAcc_Physics
    simulationSpace=BaseBoneSpace
    alphaBoolBlend {blendOption=Linear}
    baseBoneRef {boneName=neck_02}
    componentAppliedLinearAccClamp=<2000 2000 2000>
    componentLinearAccScale=<0.45 0.4 0.4>
    componentLinearVelScale=<0.45 0.4 0.4>
    overrideWorldGravity {z=-980}
   }
   n67 = Anim Dynamics - Bone: mstr_earring(ComponentPose=@n66, Alpha=1, ExternalForce="0,0,0") {
    angularDampingOverride=1
    angularSpringConstant=20000
    bAngularSpring=true
    bOverrideAngularDamping=true
    bUseGravityOverride=true
    bUsePlanarLimit=false
    bUseSphericalLimits=true
    gravityScale=0.35
    simulationSpace=RootRelative
    alphaBoolBlend {blendOption=Linear}
    boundBone {boneName=mstr_earring}
    componentLinearAccScale=<1 1 1>
    componentLinearVelScale=<1 1 1>
    gravityOverride {z=-980}
    physicsBodyDefinitions [{
      boundBone {boneName=mstr_earring}
      boxExtents=<20 30 30>
      constraintSetup {angularLimitsMax=<15 15 15>, angularLimitsMin=<-15 -15 -15>, angularTarget=<1 0 0>}
      localJointOffset=<4 0 0>
     }]
    planarLimits [{drivingBone {boneName=head}, planeTransform {location=<0 0 10>, rotation=<rot 0 0 0>}}]
    sphericalLimits [{limitRadius=7.5, drivingBone {boneName=head}, sphereLocalOffset=<0 0 0>}]
   }
   n68 = Anim Dynamics - Bone: dyn_earring_01(ComponentPose=@n67, Alpha=1, ExternalForce="0,0,0") {
    angularDampingOverride=1
    angularSpringConstant=20000
    bAngularSpring=true
    bOverrideAngularDamping=true
    bUseGravityOverride=true
    bUseSphericalLimits=true
    gravityScale=0.35
    simulationSpace=RootRelative
    alphaBoolBlend {blendOption=Linear}
    boundBone {boneName=dyn_earring_01}
    componentLinearAccScale=<1 1 1>
    componentLinearVelScale=<1 1 1>
    gravityOverride {z=-980}
    physicsBodyDefinitions [{
      boundBone {boneName=dyn_earring_01}
      boxExtents=<20 30 30>
      constraintSetup {angularLimitsMax=<15 15 15>, angularLimitsMin=<-15 -15 -15>, angularTarget=<1 0 0>}
      localJointOffset=<4 0 0>
     }]
    planarLimits [{drivingBone {boneName=head}, planeTransform {location=<0 0 0>, rotation=<rot 0 0 0>}}]
    sphericalLimits [{limitRadius=7.5, drivingBone {boneName=head}, sphereLocalOffset=<0 0 0>}]
   }
   Save cached pose 'Rban'(Pose=Component To Local(ComponentPose=@n68)) {cacheName=Rban}
   n69 = Get Look Left() -> (Look Left:double)
   n70 = "Transform (Modify) Bone - Bone: mstr_hair"!(
    ComponentPose=Local To Component(LocalPose=Use cached pose 'Rban'())
    Alpha:float=@n69
    Translation="-0.065,-0.133,0.938"
    Rotation="0.241,-1.368,12"
    Scale="1,1,1"
   ) -> (Pose) {
    rotationMode=BMM_Additive
    rotationSpace=BCS_BoneSpace
    scaleMode=BMM_Replace
    scaleSpace=BCS_BoneSpace
    translationMode=BMM_Additive
    translationSpace=BCS_BoneSpace
    alphaBoolBlend {blendInTime=0.05, blendOption=Linear, blendOutTime=0.05}
    alphaScaleBiasClamp {interpSpeedDecreasing=15, interpSpeedIncreasing=15}
    boneToModify {boneName=mstr_hair}
    rotation=<rot 0.241 -1.368 12>
    translation=<-0.065 -0.133 0.938>
   }
   n71 = "Transform (Modify) Bone - Bone: dyn_hair_midle_02"(ComponentPose=@n70, Alpha=@n69, Translation="-0.065,-0.133,0.938", Rotation="-3.765,3.734,33.231", Scale="1,1,1") {
    rotationMode=BMM_Additive
    rotationSpace=BCS_BoneSpace
    scaleMode=BMM_Replace
    scaleSpace=BCS_BoneSpace
    translationMode=BMM_Additive
    translationSpace=BCS_BoneSpace
    alphaBoolBlend {blendInTime=0.05, blendOption=Linear, blendOutTime=0.05}
    alphaScaleBiasClamp {interpSpeedDecreasing=15, interpSpeedIncreasing=15}
    boneToModify {boneName=dyn_hair_midle_02}
    rotation=<rot -3.765 3.734 33.231>
    translation=<-0.065 -0.133 0.938>
   }
   n72 = "Transform (Modify) Bone - Bone: dyn_hair_right1_02"(ComponentPose=@n71, Alpha=@n69, Translation="-0.065,-0.133,0.938", Rotation="-5,5,5", Scale="1,1,1") {
    rotationMode=BMM_Additive
    rotationSpace=BCS_BoneSpace
    scaleMode=BMM_Replace
    scaleSpace=BCS_BoneSpace
    translationMode=BMM_Additive
    translationSpace=BCS_BoneSpace
    alphaBoolBlend {blendInTime=0.05, blendOption=Linear, blendOutTime=0.05}
    alphaScaleBiasClamp {interpSpeedDecreasing=15, interpSpeedIncreasing=15}
    boneToModify {boneName=dyn_hair_right1_02}
    rotation=<rot -5 5 5>
    translation=<-0.065 -0.133 0.938>
   }
   n73 = "Transform (Modify) Bone - Bone: mstr_hair"(ComponentPose=@n72, Alpha=1, Translation="-1,0,0", Rotation="4,0,0", Scale="1,1,1") {
    rotationMode=BMM_Additive
    rotationSpace=BCS_BoneSpace
    scaleMode=BMM_Replace
    scaleSpace=BCS_BoneSpace
    translationMode=BMM_Additive
    translationSpace=BCS_BoneSpace
    alphaBoolBlend {blendInTime=0.05, blendOption=Linear, blendOutTime=0.05}
    alphaScaleBiasClamp {interpSpeedDecreasing=15, interpSpeedIncreasing=15}
    boneToModify {boneName=mstr_hair}
    rotation {pitch=4}
    translation {x=-1}
   }
   n74 = "Transform (Modify) Bone - Bone: dyn_hair_midle_02"(ComponentPose=@n73, Alpha=1, Translation="-1,0,0", Rotation="-9.999,0.190,-0.020", Scale="1,1,1") {
    rotationMode=BMM_Additive
    rotationSpace=BCS_BoneSpace
    scaleMode=BMM_Replace
    scaleSpace=BCS_BoneSpace
    translationMode=BMM_Additive
    translationSpace=BCS_BoneSpace
    alphaBoolBlend {blendInTime=0.05, blendOption=Linear, blendOutTime=0.05}
    alphaScaleBiasClamp {interpSpeedDecreasing=15, interpSpeedIncreasing=15}
    boneToModify {boneName=dyn_hair_midle_02}
    rotation=<rot -9.999 0.19 -0.02>
    translation {x=-1}
   }
   n75 = "Transform (Modify) Bone - Bone: dyn_hair_midle_04"(ComponentPose=@n74, Alpha=1, Translation="0,0,0.5", Rotation="15.000,-0.125,-0.001", Scale="1,1,1") {
    rotationMode=BMM_Additive
    rotationSpace=BCS_BoneSpace
    scaleMode=BMM_Replace
    scaleSpace=BCS_BoneSpace
    translationMode=BMM_Additive
    translationSpace=BCS_BoneSpace
    alphaBoolBlend {blendInTime=0.05, blendOption=Linear, blendOutTime=0.05}
    alphaScaleBiasClamp {interpSpeedDecreasing=15, interpSpeedIncreasing=15}
    boneToModify {boneName=dyn_hair_midle_04}
    rotation=<rot 15 -0.125 -0.001>
    translation {z=0.5}
   }
   n76 = "Transform (Modify) Bone - Bone: dyn_hair_right1_02"(ComponentPose=@n75, Alpha=1, Translation="-2,0,0", Rotation="-17.442,5.813,24.594", Scale="1,1,1") {
    rotationMode=BMM_Additive
    rotationSpace=BCS_BoneSpace
    scaleMode=BMM_Replace
    scaleSpace=BCS_BoneSpace
    translationMode=BMM_Additive
    translationSpace=BCS_BoneSpace
    alphaBoolBlend {blendInTime=0.2, blendOption=Sinusoidal, blendOutTime=0.4}
    alphaScaleBiasClamp {interpSpeedDecreasing=15, interpSpeedIncreasing=15}
    boneToModify {boneName=dyn_hair_right1_02}
    rotation=<rot -17.442 5.813 24.594>
    translation {x=-2}
   }
   n77 = "Transform (Modify) Bone - Bone: mstr_hair"(ComponentPose=@n76, Alpha=Get LookUp(), Translation="0,0,0", Rotation="15.000,0.000,0.000", Scale="1,1,1") {
    rotationMode=BMM_Additive
    rotationSpace=BCS_BoneSpace
    scaleMode=BMM_Replace
    scaleSpace=BCS_BoneSpace
    translationMode=BMM_Additive
    translationSpace=BCS_BoneSpace
    alphaBoolBlend {blendInTime=0.2, blendOption=Sinusoidal, blendOutTime=0.4}
    alphaScaleBiasClamp {bInterpResult=true, interpSpeedDecreasing=15, interpSpeedIncreasing=15}
    boneToModify {boneName=mstr_hair}
    rotation=<rot 15 0 0>
   }
   n78 = "Transform (Modify) Bone - Bone: mstr_hair_back"!(ComponentPose=@n77, Alpha:float=Get LookUp(), Translation="0,0,0.7", Rotation="5.000,0.000,0.000", Scale="1,1,1") -> (Pose) {
    rotationMode=BMM_Additive
    rotationSpace=BCS_BoneSpace
    scaleMode=BMM_Replace
    scaleSpace=BCS_BoneSpace
    translationMode=BMM_Additive
    translationSpace=BCS_BoneSpace
    alphaBoolBlend {blendInTime=0.2, blendOption=Sinusoidal, blendOutTime=0.4}
    alphaScaleBiasClamp {bInterpResult=true, interpSpeedDecreasing=15, interpSpeedIncreasing=15}
    boneToModify {boneName=mstr_hair_back}
    rotation=<rot 5 0 0>
    translation {z=0.7}
   }
   n79 = Get LookUp()
   n80 = "Transform (Modify) Bone - Bone: mstr_hair_right1"!(ComponentPose=@n78, Alpha:float=@n79, Translation="-2,0,0", Rotation="9.713,10.231,20.164", Scale="1,1,1") -> (Pose) {
    rotationMode=BMM_Additive
    rotationSpace=BCS_BoneSpace
    scaleMode=BMM_Replace
    scaleSpace=BCS_BoneSpace
    translationMode=BMM_Additive
    translationSpace=BCS_BoneSpace
    alphaBoolBlend {blendInTime=0.2, blendOption=Sinusoidal, blendOutTime=0.4}
    alphaScaleBiasClamp {bInterpResult=true, interpSpeedDecreasing=15, interpSpeedIncreasing=15}
    boneToModify {boneName=mstr_hair_right1}
    rotation=<rot 9.713 10.231 20.164>
    translation {x=-2}
   }
   n81 = "Transform (Modify) Bone - Bone: dyn_hair_right1_02"(ComponentPose=@n80, Alpha=@n79, Translation="-2,0,0", Rotation="6,3.026,20.252", Scale="1,1,1") {
    rotationMode=BMM_Additive
    rotationSpace=BCS_BoneSpace
    scaleMode=BMM_Replace
    scaleSpace=BCS_BoneSpace
    translationMode=BMM_Additive
    translationSpace=BCS_BoneSpace
    alphaBoolBlend {blendInTime=0.2, blendOption=Sinusoidal, blendOutTime=0.4}
    alphaScaleBiasClamp {bInterpResult=true, interpSpeedDecreasing=15, interpSpeedIncreasing=15}
    boneToModify {boneName=dyn_hair_right1_02}
    rotation=<rot 6 3.026 20.252>
    translation {x=-2}
   }
   n82 = "Transform (Modify) Bone - Bone: mstr_hair_midle"!(ComponentPose=@n81, Alpha:float=Get LookUp(), Translation="0,1.127,0", Rotation="8.935,4.400,37.643", Scale="1,1,1") -> (Pose) {
    rotationMode=BMM_Additive
    rotationSpace=BCS_BoneSpace
    scaleMode=BMM_Replace
    scaleSpace=BCS_BoneSpace
    translationMode=BMM_Additive
    translationSpace=BCS_BoneSpace
    alphaBoolBlend {blendInTime=0.2, blendOption=Sinusoidal, blendOutTime=0.4}
    alphaScaleBiasClamp {bInterpResult=true, interpSpeedDecreasing=15, interpSpeedIncreasing=15}
    boneToModify {boneName=mstr_hair_midle}
    rotation=<rot 8.935 4.4 37.643>
    translation {y=1.127}
   }
   n83 = "Transform (Modify) Bone - Bone: mstr_hair_left2"!(ComponentPose=@n82, Alpha:float=Get LookUp(), Translation="-3,0,0", Rotation="0,-2,0", Scale="1,1,1") -> (Pose) {
    rotationMode=BMM_Additive
    rotationSpace=BCS_BoneSpace
    scaleMode=BMM_Replace
    scaleSpace=BCS_BoneSpace
    translationMode=BMM_Additive
    translationSpace=BCS_BoneSpace
    alphaBoolBlend {blendInTime=0.2, blendOption=Sinusoidal, blendOutTime=0.4}
    alphaScaleBiasClamp {bInterpResult=true, interpSpeedDecreasing=15, interpSpeedIncreasing=15}
    boneToModify {boneName=mstr_hair_left2}
    rotation {yaw=-2}
    translation {x=-3}
   }
   n84 = "Transform (Modify) Bone - Bone: dyn_hair_midle_02"(ComponentPose=@n83, Alpha=Get LookUp(), Translation="0,0,0", Rotation="2,-2,0", Scale="1,1,1") {
    rotationMode=BMM_Additive
    rotationSpace=BCS_BoneSpace
    scaleMode=BMM_Replace
    scaleSpace=BCS_BoneSpace
    translationMode=BMM_Additive
    translationSpace=BCS_BoneSpace
    alphaBoolBlend {blendInTime=0.2, blendOption=Sinusoidal, blendOutTime=0.4}
    alphaScaleBiasClamp {bInterpResult=true, interpSpeedDecreasing=15, interpSpeedIncreasing=15}
    boneToModify {boneName=dyn_hair_midle_02}
    rotation {pitch=2, yaw=-2}
   }
   n85 = "Transform (Modify) Bone - Bone: mstr_hair"(ComponentPose=@n84, Alpha=Get LookDown(), Translation="0,0,0", Rotation="-20.000,0.000,0.000", Scale="1,1,1") {
    rotationMode=BMM_Additive
    rotationSpace=BCS_BoneSpace
    scaleMode=BMM_Replace
    scaleSpace=BCS_BoneSpace
    translationMode=BMM_Additive
    translationSpace=BCS_BoneSpace
    alphaBoolBlend {blendInTime=0.2, blendOption=Sinusoidal, blendOutTime=0.4}
    alphaScaleBiasClamp {bInterpResult=true, interpSpeedDecreasing=15, interpSpeedIncreasing=15}
    boneToModify {boneName=mstr_hair}
    rotation=<rot -20.0001 0 0>
   }
   n86 = "Transform (Modify) Bone - Bone: mstr_hair_back"(ComponentPose=@n85, Alpha=Get LookDown(), Translation="0.289,0.293,2.619", Rotation="-16.858,-25.167,22.512", Scale="1,1,1") {
    rotationMode=BMM_Additive
    rotationSpace=BCS_BoneSpace
    scaleMode=BMM_Replace
    scaleSpace=BCS_BoneSpace
    translationMode=BMM_Additive
    translationSpace=BCS_BoneSpace
    alphaBoolBlend {blendInTime=0.2, blendOption=Sinusoidal, blendOutTime=0.4}
    alphaScaleBiasClamp {bInterpResult=true, interpSpeedDecreasing=15, interpSpeedIncreasing=15}
    boneToModify {boneName=mstr_hair_back}
    rotation=<rot -16.8581 -25.167 22.5121>
    translation=<0.289 0.293 2.619>
   }
   n87 = "Transform (Modify) Bone - Bone: mstr_hair_left3"!(ComponentPose=@n86, Alpha:float=Get LookDown(), Translation="0,0,0", Rotation="15,-15,0", Scale="1,1,1") -> (Pose) {
    rotationMode=BMM_Additive
    rotationSpace=BCS_BoneSpace
    scaleMode=BMM_Replace
    scaleSpace=BCS_BoneSpace
    translationMode=BMM_Additive
    translationSpace=BCS_BoneSpace
    alphaBoolBlend {blendInTime=0.2, blendOption=Sinusoidal, blendOutTime=0.4}
    alphaScaleBiasClamp {bInterpResult=true, interpSpeedDecreasing=15, interpSpeedIncreasing=15}
    boneToModify {boneName=mstr_hair_left3}
    rotation {pitch=15, yaw=-15}
   }
   n88 = "Transform (Modify) Bone - Bone: mstr_hair_left2"(ComponentPose=@n87, Alpha=Get LookDown(), Translation="-1,1,0", Rotation="-0.000,2.174,-0.000", Scale="1,1,1") {
    rotationMode=BMM_Additive
    rotationSpace=BCS_BoneSpace
    scaleMode=BMM_Replace
    scaleSpace=BCS_BoneSpace
    translationMode=BMM_Additive
    translationSpace=BCS_BoneSpace
    alphaBoolBlend {blendInTime=0.2, blendOption=Sinusoidal, blendOutTime=0.4}
    alphaScaleBiasClamp {bInterpResult=true, interpSpeedDecreasing=15, interpSpeedIncreasing=15}
    boneToModify {boneName=mstr_hair_left2}
    rotation=<rot 0 2.174 0>
    translation=<-1 1>
   }
   Save cached pose 'Clipping Fix '(Pose=Component To Local(ComponentPose=@n88)) {cacheName="Clipping Fix "}
  ]}
 EventGraph {nodes [
   Event BlueprintUpdateAnimation()
   > n1 = Sequence!(execute) -> (then_0, then_1, then_2:Exec, then_3:Exec, then_4:Exec)
   n2 = AND Boolean(A=Get bIsCrouching(), B=Get bIsMoving2D())
   n3 = Get bIsCrouching()
   > Set NoAimStates(NoAimStates:bool=NOT Boolean(A=OR Boolean(A=@n2, B=@n3))) -> (Output_Get:bool)
   n4 = OR Boolean(A=AND Boolean(A=Get bIsCrouching(), B=Get bIsMoving2D()), B=Get bIsCrouching())
   > Set TargetingStates(TargetingStates:bool=AND Boolean(A=Get bIsTargeting(), B=NOT Boolean(A=@n4))) -> (Output_Get:bool)
   n5 = AND Boolean(A=Get bIsCrouching(), B=Get bIsMoving2D())
   > Set CrouchingAndMovingStates(CrouchingAndMovingStates:bool=@n5) -> (Output_Get:bool)
   n6 = Get bIsMoving2D()
   > Set CrouchingStates(CrouchingStates:bool=AND Boolean(A=Get bIsCrouching(), B=NOT Boolean(A=@n6))) -> (Output_Get:bool)
   n7 = Get bIsCrouching()
   n8 = Get bIsMoving2D()
   > Set CrouchingAndMovingStates_Targeting(CrouchingAndMovingStates_Targeting:bool=AND Boolean(A=AND Boolean(A=@n7, B=@n8), B=Get bIsTargeting())) -> (Output_Get:bool)
   n9 = Get bIsSkyDiving()
   > Set IsFallnigAfterJump(IsFallnigAfterJump:bool=AND Boolean(A=Get bIsFalling(), B=NOT Boolean(A=@n9))) -> (Output_Get:bool)
   > Set CrouchingAndTargeting(CrouchingAndTargeting:bool=AND Boolean(A=Get bIsCrouching(), B=Get bIsTargeting())) -> (Output_Get:bool)
   > Set IsFallnigAfterJumpWithBackpack(IsFallnigAfterJumpWithBackpack:bool=AND Boolean(A=Get IsFallnigAfterJump() -> (IsFallnigAfterJump:bool), B=Get IsWearingBackpack() -> (IsWearingBackpack:bool))) -> (Output_Get:bool)
   n10 = Get IsWearingBackpack()
   > Set IsFallnigAfterJumpWithoutBackpack(IsFallnigAfterJumpWithoutBackpack:bool=AND Boolean(A=Get IsFallnigAfterJump(), B=NOT Boolean(A=@n10))) -> (Output_Get:bool)
   > Set SkydiwingWithBackpac(SkydiwingWithBackpac:bool=AND Boolean(A=Get bIsSkyDiving(), B=Get IsWearingBackpack())) -> (Output_Get:bool)
   n11 = Get bIsParachuteOpened()
   > Set SkydiwingButNotParachute(SkydiwingButNotParachute:bool=AND Boolean(A=Get bIsSkyDiving(), B=NOT Boolean(A=@n11))) -> (Output_Get:bool)
   n12 = GetOwnerSkeletalMeshComponent()
   Set LookDown(
    execute=@n1.then_1
    LookDown:float=float * float(
     A=GetAlphaFromReferenceTransformDelta(SkeletalMeshComponent=@n12, BoneName=spine_05, BaseName=head, InputRange=(X=-10,Y=-20), TwistAxis="1,0,0")
     B:bool=Get NoAimStates() -> (NoAimStates:bool)
    ) {is=title K2Node_PromotableOperator}
   ) -> (Output_Get:float)
   n13 = GetOwnerSkeletalMeshComponent()
   > Set LookUp(LookUp:float=float * float(
     A=GetAlphaFromReferenceTransformDelta(SkeletalMeshComponent=@n13, BoneName=spine_05, BaseName=head, InputRange=(X=10,Y=45), TwistAxis="1,0,0")
     B=Get NoAimStates()
    )) -> (Output_Get:float)
   n14 = GetOwnerSkeletalMeshComponent()
   > Set LookUpTargeting(LookUpTargeting:float=float * float(
     A=GetAlphaFromReferenceTransformDelta(SkeletalMeshComponent=@n14, BoneName=spine_05, BaseName=head, InputRange=(X=7,Y=45), TwistAxis="1,0,0")
     B=Get TargetingStates() -> (TargetingStates:bool)
    )) -> (Output_Get:float)
   n15 = GetOwnerSkeletalMeshComponent()
   n16 = GetAlphaFromReferenceTransformDelta(SkeletalMeshComponent=@n15, BoneName=clavicle_r, BaseName=head, InputRange=(X=-60,Y=-90), TwistAxis="1,0,0")
   n17 = GetAlphaFromReferenceTransformDelta(SkeletalMeshComponent=@n15, BoneName=clavicle_r, BaseName=head, InputRange=(X=60,Y=90), TwistAxis="1,0,0")
   n18 = Get LookUp()
   > Set ClavRoataionMelleFixAlpha(ClavRoataionMelleFixAlpha:float=float * float(A=float + float(A=@n17, B=@n16), B:double=float + float!(A=@n17, B=@n16, C:float=@n18) -> (ReturnValue))) -> (Output_Get:float)
   > Set LookDownTargeting(LookDownTargeting:double=float * float(A=Get LookDown(), B:bool=Get TargetingStates())) -> (Output_Get:double)
   n19 = GetOwnerSkeletalMeshComponent()
   > Set LookUpCrouchingAndMoving(LookUpCrouchingAndMoving:double=float * float(
     A=GetAlphaFromReferenceTransformDelta(SkeletalMeshComponent=@n19, BoneName=spine_05, BaseName=head, InputRange=(X=40,Y=60), TwistAxis="1,0,0")
     B=Get CrouchingAndMovingStates() -> (CrouchingAndMovingStates:bool)
    )) -> (Output_Get:double)
   > Set LookDownCrouchingAndMoving(LookDownCrouchingAndMoving:double=float * float(B=Get CrouchingAndMovingStates())) -> (Output_Get:double)
   n20 = GetOwnerSkeletalMeshComponent()
   > Set LookUpCrouchingState(LookUpCrouchingState:double=float * float(
     A=GetAlphaFromReferenceTransformDelta(SkeletalMeshComponent=@n20, BoneName=spine_05, BaseName=head, InputRange=(X=10,Y=45), TwistAxis="1,0,0")
     B=Get CrouchingStates() -> (CrouchingStates:bool)
    )) -> (Output_Get:double)
   n21 = GetOwnerSkeletalMeshComponent()
   > Set LookDownCrouching(LookDownCrouching:double=float * float(
     A=GetAlphaFromReferenceTransformDelta(SkeletalMeshComponent=@n21, BoneName=spine_05, BaseName=head, InputRange=(X=15,Y=-4), TwistAxis="1,0,0")
     B=Get CrouchingStates()
    )) -> (Output_Get:double)
   n22 = GetOwnerSkeletalMeshComponent()
   > Set LookUpCrouchingAndMoving_Targeting(LookUpCrouchingAndMoving_Targeting:double=float * float(
     A=GetAlphaFromReferenceTransformDelta(SkeletalMeshComponent=@n22, BoneName=spine_05, BaseName=head, InputRange=(X=65,Y=80), TwistAxis="1,0,0")
     B=Get CrouchingAndMovingStates_Targeting() -> (CrouchingAndMovingStates_Targeting:bool)
    )) -> (Output_Get:double)
   n23 = GetAlphaFromReferenceTransformDelta(SkeletalMeshComponent=GetOwnerSkeletalMeshComponent(), BoneName=spine_05, BaseName=head, InputRange=(X=20,Y=40), TwistAxis="1,0,0")
   n24 = "float < float"(A=@n23.RawOutput, B=40)
   n25 = "float > float"(A=@n23.RawOutput, B=20)
   > Set MiddleCrouchingAndMoving(MiddleCrouchingAndMoving:double=float * float!(A=@n23, B=Get CrouchingAndMovingStates(), C:bool=AND Boolean(A=@n25, B=@n24)) -> (ReturnValue)) -> (Output_Get:double)
   > Set LookUpCorunchingAndTargeting(LookUpCorunchingAndTargeting:double=float * float!(A=Get LookUpCrouchingState(), B=Get CrouchingAndTargeting() -> (CrouchingAndTargeting:bool)) -> (ReturnValue)) -> (Output_Get:double)
   > Set LookDownCorunchingAndTargeting(LookDownCorunchingAndTargeting:double=float * float(A=Get LookDownCrouching(), B=Get CrouchingAndTargeting())) -> (Output_Get:double)
   n26 = GetOwnerSkeletalMeshComponent()
   > Set LookUpTargeting_Pistol(LookUpTargeting_Pistol:double=float * float(
     A=GetAlphaFromReferenceTransformDelta(SkeletalMeshComponent=@n26, BoneName=spine_05, BaseName=head, InputRange=(X=7,Y=45), TwistAxis="1,0,0")
     B=Get PistolTargeting() -> (PistolTargeting:bool)
    )) -> (Output_Get:double)
   n27 = Branch(execute=@n1.then_2, Condition=Get IsWearingBackpack())
   n28 = GetGameTimeInSeconds()
   n29 = Self-Reference()
   > n30 = Gravity Override State Function DEPRECATED(
    CharacterPartAnimInstance=@n29
    GameTimeInSeconds=@n28
    GravityOverrideParams=Get Gravity Override Params_Backpack() -> (Gravity Override Params_Backpack:Gravity Override Params Struct Structure)
   )
   > Set Gravity Override(Gravity Override=@n30)
   n31 = SelectVector(A="0.05,0.05,0.05", B="0.15,0.15,0.15", bPickA=Get bIsSkyDiving())
   > Set Component Linear Acc Scale(Component Linear Acc Scale:Vector=@n31) -> (Output_Get:Vector)
   > Set Component Linear Vel Scale(Component Linear Vel Scale:Vector=@n31) -> (Output_Get:Vector)
   n32 = Gravity Override State Function DEPRECATED(execute=@n27.else, CharacterPartAnimInstance=@n29, GameTimeInSeconds=@n28, GravityOverrideParams=Get Gravity Override Params())
   > Set Gravity Override(Gravity Override=@n32)
   n33 = SelectVector(A="0.2,0.1,0.3", B="0.3,0.25,0.3", bPickA=Get bIsSkyDiving())
   > Set Component Linear Acc Scale(Component Linear Acc Scale=@n33)
   > Set Component Linear Vel Scale(Component Linear Vel Scale=@n33)
   n34 = Cast To FortPlayerAnimInstance(execute=@n1.then_3, Object=GetAnimInstance(self=GetOwnerSkeletalMeshComponent()))
   > n35 = Switch on EFortWeaponCoreAnimation!(execute, Selection:EFortWeaponCoreAnimation Enum=Get WeaponCoreAnim(self=@n34)) -> (
    Melee:Exec
    Pistol:Exec
    Shotgun:Exec
    PaperBlueprint:Exec
    Rifle:Exec
    MeleeOneHand:Exec
    MachinePistol:Exec
    RocketLauncher:Exec
    GrenadeLauncher:Exec
    GoingCommando:Exec
    AssaultRifle:Exec
    TacticalShotgun:Exec
    SniperRifle:Exec
    TrapPlacement:Exec
    ShoulderLauncher:Exec
    AbilityDecoTool:Exec
    Crossbow:Exec
    C4:Exec
    RemoteControl:Exec
    DualWield:Exec
    AR_BullPup:Exec
    AR_ForwardGrip:Exec
    MedPackPaddles:Exec
    SMG_P90:Exec
    AR_DrumGun:Exec
    Consumable_Small:Exec
    Consumable_Large:Exec
    Balloon:Exec
    MountedTurret:Exec
    CreativeTool:Exec
    ExplosiveBow:Exec
    AshtonIndigo:Exec
    AshtonChicago:Exec
    MeleeDualWield:Exec
    Unarmed:Exec
   )
   n36 = Branch(execute=@n35.Pistol, Condition=Get TargetingStates())
   > Set PistolTargeting(PistolTargeting:bool=true) -> (Output_Get:bool)
   Set PistolTargeting(execute=@n36.else, PistolTargeting=false)
   n37 = GetOwnerSkeletalMeshComponent()
   Set Look Left(
    execute=@n1.then_4
    Look Left:double=float * float(
     A=GetAlphaFromReferenceTransformDelta(SkeletalMeshComponent=@n37, BoneName=spine_05, BaseName=head, DeltaComponentType=EulerX, InputRange=(X=0,Y=-18), TwistAxis="1,0,0")
     B=Get NoAimStates()
    )
   ) -> (Output_Get:double)
   Event OnCharacterCustomizationLoaded()
   n38 = CheckForBackbling()
   > Set IsWearingBackpack(IsWearingBackpack:bool=OR Boolean!(A=@n38, B=@n38.bIsBackblingACape, C:bool=@n38.bIsBackblingLong) -> (ReturnValue)) -> (Output_Get:bool)
   Print Float!(
    Execute:Exec
    Float:double=Get LookDownCrouching()
    OptionalString:String=LookDownCrouching
    Duration:float=2
    TextColor:Linear Color Structure=(R=0,G=0.6588,B=1,A=1)
   ) -> (Then:Exec) {is=title K2Node_MacroInstance}
   > Print Float(Float=Get LookUpCrouchingState(), OptionalString=LookUpCrouching, Duration=2, TextColor=(R=0,G=0.6588,B=1,A=1))
   > Print Float(Float=Get LookUpTargeting(), OptionalString=LookUpTargeting, Duration=2, TextColor=(R=0,G=0.6588,B=1,A=1))
   > Print Float(Float=Get LookDownTargeting(), OptionalString=LookDownTargeting, Duration=2, TextColor=(R=0,G=0.6588,B=1,A=1))
   > Print Float(Float=Get LookUp(), OptionalString=LookUp, Duration=2, TextColor=(R=0,G=0.6588,B=1,A=1))
   > Print Float(Float=Get LookDown(), OptionalString=LookDown, Duration=2, TextColor=(R=0,G=0.6588,B=1,A=1))
   > Print Float(Float=Get LookUpCrouchingAndMoving(), OptionalString=LookUpCrouchingAndMoving, Duration=2, TextColor=(R=0,G=0.6588,B=1,A=1))
   > Print Float(
    Float=Get LookDownCrouchingAndMoving() -> (LookDownCrouchingAndMoving:double)
    OptionalString=LookDownCrouchingAndMoving
    Duration=2
    TextColor=(R=0,G=0.6588,B=1,A=1)
   )
   > Print Float(
    Float=Get LookUpCrouchingAndMoving_Targeting()
    OptionalString=LookUpCrouchingAndMoving_Targeting
    Duration=2
    TextColor=(R=0,G=0.6588,B=1,A=1)
   )
   > Print Float(Float=Get MiddleCrouchingAndMoving(), OptionalString=MiddleCrouchingAndMoving, Duration=2, TextColor=(R=0,G=0.6588,B=1,A=1))
   > Print Float(Float=Get LookUpCorunchingAndTargeting(), OptionalString=LookUpCorunchingAndTargeting, Duration=2, TextColor=(R=0,G=0.6588,B=1,A=1))
   > Print Float(Float=Get LookDownCorunchingAndTargeting(), OptionalString=LookDownCorunchingAndTargeting, Duration=2, TextColor=(R=0,G=0.6588,B=1,A=1))
   > Print Bool!(
    Execute:Exec
    Bool:bool=Get IsWearingBackpack()
    OptionalString:String="IsWearingBackpack "
    Duration:float=2
    TextColor:Linear Color Structure=(R=0,G=0.6588,B=1,A=1)
   ) -> (Then:Exec) {is=title K2Node_MacroInstance}
   > Print Bool(Bool=Get IsFallnigAfterJump(), OptionalString="IsFallnigAfterJump ", Duration=2, TextColor=(R=0,G=0.6588,B=1,A=1))
   > Print Bool(Bool=Get IsFallnigAfterJumpWithBackpack(), OptionalString="IsFallnigAfterJumpWithBackpack]", Duration=2, TextColor=(R=0,G=0.6588,B=1,A=1))
   > Print Bool(Bool=Get IsFallnigAfterJumpWithoutBackpack(), OptionalString=IsFallnigAfterJumpWithoutBackpack, Duration=2, TextColor=(R=0,G=0.6588,B=1,A=1))
   > Print Float(OptionalString=------------------, Duration=2, TextColor=(R=0,G=0.6588,B=1,A=1))
   PrintString!(execute, InString=Hello, bPrintToScreen=false, bPrintToLog=false, TextColor, Duration, Key) -> (then)
   AND Boolean(A=@n5, B=NOT Boolean(A=Get bIsTargeting()))
  ]}
]
variables [
 ClavRoataionMelleFixAlpha {type=number}
 Component Linear Acc Scale {type=Vector, default=<0.45 0.4 0.4>}
 Component Linear Vel Scale {type=Vector, default=<0.45 0.4 0.4>}
 CrouchingAndMovingStates {type=boolean}
 CrouchingAndMovingStates_Targeting {type=boolean}
 CrouchingAndTargeting {type=boolean}
 CrouchingStates {type=boolean}
 Gravity Override {type=Vector}
 Gravity Override Params {
  type=GravityOverrideParamsStruct
  default {
   skydive_DivingDown {windAmplitude=<200 -256 40>}
   skydive_FloatingDown {gravityOverride=<0 0 600>, windAmplitude=<200 -256 40>}
   skydive_GoingUp {gravityOverride=<0 0 700>, windAmplitude=<200 -256 40>}
  }
 }
 Gravity Override Params_Backpack {
  type=GravityOverrideParamsStruct
  default {
   skydive_DivingDown {windAmplitude=<200 -256 40>}
   skydive_FloatingDown {gravityOverride=<0 0 0>, windAmplitude=<20 -25 40>, windOffset=<0 0 0>}
   skydive_GoingUp {gravityOverride=<0 0 0>, windAmplitude=<20 -20 1>, windFrequency=<20 7 8>}
   skydive_Parachuting {gravityOverride=<0 0 -100>, windAmplitude=<20 -60 40>}
  }
 }
 Head {type=SkeletalMeshComponent}
 IsFallnigAfterJump {type=boolean}
 IsFallnigAfterJumpWithBackpack {type=boolean}
 IsFallnigAfterJumpWithoutBackpack {type=boolean}
 IsWearingBackpack {type=boolean}
 Look Left {type=number}
 LookDown {type=number}
 LookDownCorunchingAndTargeting {type=number}
 LookDownCrouching {type=number}
 LookDownCrouchingAndMoving {type=number}
 LookDownTargeting {type=number}
 LookUp {type=number}
 LookUpCorunchingAndTargeting {type=number}
 LookUpCrouchingAndMoving {type=number}
 LookUpCrouchingAndMoving_Targeting {type=number}
 LookUpCrouchingState {type=number}
 LookUpTargeting {type=number}
 LookUpTargeting_Pistol {type=number}
 MiddleCrouchingAndMoving {type=number}
 NewVar {type=boolean}
 NoAimStates {type=boolean}
 PistolTargeting {type=boolean}
 SkydiwingButNotParachute {type=boolean}
 SkydiwingWithBackpac {type=boolean}
 TargetingStates {type=boolean}
]
